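# Physics asset for the female medium MysticFang player cosmetic body, covering its skirt plus pelvis and thigh capsules.
# Eight MSTR skirt anchors are angularly locked to the pelvis, each heading a three-link chain of dyn capsules around the front, sides and back.
# Chain joints have tight swing and twist limits, with right-side chains tuned differently from left and middle ones.
# Thigh-to-pelvis joints swing and twist freely.
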PhysicsAsset /BRCosmetics/Characters/Player/Female/Medium/Bodies/F_MED_MysticFang/Meshes/F_MED_MysticFang_Physics
bodies {
 MSTR_BackSkirt_l [{shapeType=Capsule, length=0.505, radius=0.505, shapeName=MSTR_BackSkirt_l_capsule, rotation=<rot 0 -90 0>}]
 MSTR_BackSkirt_m [{shapeType=Capsule, length=0.505, radius=0.505, shapeName=MSTR_BackSkirt_m_capsule, rotation=<rot 0 -90 0>}]
 MSTR_BackSkirt_r [{shapeType=Capsule, length=0.505, radius=0.505, shapeName=MSTR_BackSkirt_r_capsule}]
 MSTR_SideSkirt_l [{shapeType=Capsule, length=0.505, radius=0.505, shapeName=MSTR_SideSkirt_l_capsule, rotation=<rot 0 -90 0>}]
 MSTR_SideSkirt_r [{shapeType=Capsule, length=0.505, radius=0.505, shapeName=MSTR_SideSkirt_r_capsule}]
 MSTR_frontSkirt_l [{shapeType=Capsule, length=0.505, radius=0.505, shapeName=MSTR_frontSkirt_l_capsule, rotation=<rot 0 -90 0>}]
 MSTR_frontSkirt_m [{shapeType=Capsule, length=0.505, radius=0.505, shapeName=MSTR_frontSkirt_m_capsule, rotation=<rot 0 -90 0>}]
 MSTR_frontSkirt_r [{shapeType=Capsule, length=0.505, radius=0.505, shapeName=MSTR_frontSkirt_r_capsule}]
 dyn_BackSkirt_l_01 [{shapeType=Capsule, length=10, radius=5, shapeName=dyn_BackSkirt_l_01_capsule, center=<8 0 2>, rotation=<rot 0 0 90>}]
 dyn_BackSkirt_l_02 [{shapeType=Capsule, length=10, radius=5, shapeName=dyn_BackSkirt_l_02_capsule, center=<8 0 2>, rotation=<rot 0 0 90>}]
 dyn_BackSkirt_l_03 [{shapeType=Capsule, length=10, radius=5, shapeName=dyn_BackSkirt_l_03_capsule, center=<8 0 -3>, rotation=<rot 0 0 90>}]
 dyn_BackSkirt_m_01 [{shapeType=Capsule, length=10, radius=5, shapeName=dyn_BackSkirt_m_01_capsule, center=<8 0 1.5>, rotation=<rot 0 0 90>}]
 dyn_BackSkirt_m_02 [{shapeType=Capsule, length=10, radius=5, shapeName=dyn_BackSkirt_m_02_capsule, center=<8 0 1>, rotation=<rot 0 0 90>}]
 dyn_BackSkirt_m_03 [{shapeType=Capsule, length=10, radius=5, shapeName=dyn_BackSkirt_m_03_capsule, center=<8 0 -4>, rotation=<rot 0 0 90>}]
 dyn_BackSkirt_r_01 [{shapeType=Capsule, length=10, radius=5, shapeName=dyn_BackSkirt_r_01_capsule, center=<8 0 2>, rotation=<rot 0 0 90>}]
 dyn_BackSkirt_r_02 [{shapeType=Capsule, length=10, radius=5, shapeName=dyn_BackSkirt_r_02_capsule, center=<8 0 2>, rotation=<rot 0 0 90>}]
 dyn_BackSkirt_r_03 [{shapeType=Capsule, length=10, radius=5, shapeName=dyn_BackSkirt_r_03_capsule, center=<8 0 -3>, rotation=<rot 0 0 90>}]
 dyn_SideSkirt_l_01 [{shapeType=Capsule, length=10, radius=5, shapeName=dyn_SideSkirt_l_01_capsule, center=<8 0 4>, rotation=<rot 0 0 90>}]
 dyn_SideSkirt_l_02 [{shapeType=Capsule, length=10, radius=5, shapeName=dyn_SideSkirt_l_02_capsule, center=<8 0 5>, rotation=<rot 0 0 90>}]
 dyn_SideSkirt_l_03 [{shapeType=Capsule, length=10, radius=5, shapeName=dyn_SideSkirt_l_03_capsule, center=<8 0 2>, rotation=<rot 0 0 90>}]
 dyn_SideSkirt_r_01 [{shapeType=Capsule, length=10, radius=5, shapeName=dyn_SideSkirt_r_01_capsule, center=<8 0 4>, rotation=<rot 0 0 90>}]
 dyn_SideSkirt_r_02 [{shapeType=Capsule, length=10, radius=5, shapeName=dyn_SideSkirt_r_02_capsule, center=<8 0 5>, rotation=<rot 0 0 90>}]
 dyn_SideSkirt_r_03 [{shapeType=Capsule, length=10, radius=5, shapeName=dyn_SideSkirt_r_03_capsule, center=<8 0 2>, rotation=<rot 0 0 90>}]
 dyn_frontSkirt_l_01 [{shapeType=Capsule, length=10, radius=5, shapeName=dyn_frontSkirt_l_01_capsule, center=<8 0 3>, rotation=<rot 0 0 90>}]
 dyn_frontSkirt_l_02 [{shapeType=Capsule, length=10, radius=5, shapeName=dyn_frontSkirt_l_02_capsule, center=<8 0 4>, rotation=<rot 0 0 90>}]
 dyn_frontSkirt_l_03 [{shapeType=Capsule, length=10, radius=5, shapeName=dyn_frontSkirt_l_03_capsule, center=<8 0 0.5>, rotation=<rot 0 0 90>}]
 dyn_frontSkirt_m_01 [{shapeType=Capsule, length=10, radius=5, shapeName=dyn_frontSkirt_m_01_capsule, center=<8 0 3>, rotation=<rot 0 0 90>}]
 dyn_frontSkirt_m_02 [{shapeType=Capsule, length=10, radius=5, shapeName=dyn_frontSkirt_m_02_capsule, center=<8 0 4>, rotation=<rot 0 0 90>}]
 dyn_frontSkirt_m_03 [{shapeType=Capsule, length=10, radius=5, shapeName=dyn_frontSkirt_m_03_capsule, center=<8 0 -2>, rotation=<rot 0 0 90>}]
 dyn_frontSkirt_r_01 [{shapeType=Capsule, length=10, radius=5, shapeName=dyn_frontSkirt_r_01_capsule, center=<8 0 3>, rotation=<rot 0 0 90>}]
 dyn_frontSkirt_r_02 [{shapeType=Capsule, length=10, radius=5, shapeName=dyn_frontSkirt_r_02_capsule, center=<8 0 4>, rotation=<rot 0 0 90>}]
 dyn_frontSkirt_r_03 [{shapeType=Capsule, length=10, radius=5, shapeName=dyn_frontSkirt_r_03_capsule, center=<8 0 0.5>, rotation=<rot 0 0 90>}]
 pelvis [{shapeType=Capsule, length=14.25, radius=9, shapeName=pelvis_capsule, center=<-2.9969 -0.6851 0>}]
 thigh_l [{
   shapeType=Capsule
   length=29.4545
   radius=8.7955
   shapeName=thigh_l_capsule
   center=<-20.3533 0.5352 0.82>
   rotation=<rot 88.8689 2.1555 -0.3167>
  }]
 thigh_r [{
   shapeType=Capsule
   length=29.4545
   radius=8.7955
   shapeName=thigh_r_capsule
   center=<20.3533 -0.5352 -0.82>
   rotation=<rot 88.8689 2.1555 179.6833>
  }]
}
constraints [
 {
  bone1Name=thigh_l
  bone2Name=pelvis
  swing1LimitDegrees=45
  swing1Motion=Free
  swing2LimitDegrees=45
  swing2Motion=Free
  twistLimitDegrees=45
  twistMotion=Free
 }
 {
  bone1Name=thigh_r
  bone2Name=pelvis
  swing1LimitDegrees=45
  swing1Motion=Free
  swing2LimitDegrees=45
  swing2Motion=Free
  twistLimitDegrees=45
  twistMotion=Free
 }
 {
  bone1Name=MSTR_BackSkirt_l
  bone2Name=pelvis
  swing1LimitDegrees=45
  swing1Motion=Locked
  swing2LimitDegrees=45
  swing2Motion=Locked
  twistLimitDegrees=45
  twistMotion=Locked
 }
 {bone1Name=dyn_BackSkirt_l_01, bone2Name=MSTR_BackSkirt_l, swing1LimitDegrees=10, swing2LimitDegrees=30, twistLimitDegrees=2}
 {bone1Name=dyn_BackSkirt_l_03, bone2Name=dyn_BackSkirt_l_02, swing1LimitDegrees=10, swing2LimitDegrees=30, twistLimitDegrees=2}
 {
  bone1Name=MSTR_BackSkirt_m
  bone2Name=pelvis
  swing1LimitDegrees=45
  swing1Motion=Locked
  swing2LimitDegrees=45
  swing2Motion=Locked
  twistLimitDegrees=45
  twistMotion=Locked
 }
 {bone1Name=dyn_BackSkirt_m_01, bone2Name=MSTR_BackSkirt_m, swing1LimitDegrees=10, swing2LimitDegrees=30, twistLimitDegrees=2}
 {bone1Name=dyn_BackSkirt_m_03, bone2Name=dyn_BackSkirt_m_02, swing1LimitDegrees=10, swing2LimitDegrees=30, twistLimitDegrees=2}
 {
  bone1Name=MSTR_SideSkirt_l
  bone2Name=pelvis
  swing1LimitDegrees=45
  swing1Motion=Locked
  swing2LimitDegrees=45
  swing2Motion=Locked
  twistLimitDegrees=45
  twistMotion=Locked
 }
 {bone1Name=dyn_SideSkirt_l_01, bone2Name=MSTR_SideSkirt_l, swing1LimitDegrees=10, swing2LimitDegrees=30, twistLimitDegrees=2}
 {bone1Name=dyn_SideSkirt_l_03, bone2Name=dyn_SideSkirt_l_02, swing1LimitDegrees=10, swing2LimitDegrees=30, twistLimitDegrees=2}
 {
  bone1Name=MSTR_frontSkirt_l
  bone2Name=pelvis
  swing1LimitDegrees=45
  swing1Motion=Locked
  swing2LimitDegrees=45
  swing2Motion=Locked
  twistLimitDegrees=45
  twistMotion=Locked
 }
 {bone1Name=dyn_frontSkirt_l_01, bone2Name=MSTR_frontSkirt_l, swing1LimitDegrees=10, swing2LimitDegrees=30, twistLimitDegrees=2}
 {bone1Name=dyn_frontSkirt_l_03, bone2Name=dyn_frontSkirt_l_02, swing1LimitDegrees=10, swing2LimitDegrees=30, twistLimitDegrees=2}
 {
  bone1Name=MSTR_frontSkirt_m
  bone2Name=pelvis
  swing1LimitDegrees=45
  swing1Motion=Locked
  swing2LimitDegrees=45
  swing2Motion=Locked
  twistLimitDegrees=45
  twistMotion=Locked
 }
 {bone1Name=dyn_frontSkirt_m_01, bone2Name=MSTR_frontSkirt_m, swing1LimitDegrees=10, swing2LimitDegrees=30, twistLimitDegrees=2}
 {bone1Name=dyn_frontSkirt_m_03, bone2Name=dyn_frontSkirt_m_02, swing1LimitDegrees=10, swing2LimitDegrees=30, twistLimitDegrees=2}
 {bone1Name=dyn_BackSkirt_l_02, bone2Name=dyn_BackSkirt_l_01, swing1LimitDegrees=10, swing2LimitDegrees=30, twistLimitDegrees=2}
 {bone1Name=dyn_BackSkirt_m_02, bone2Name=dyn_BackSkirt_m_01, swing1LimitDegrees=10, swing2LimitDegrees=30, twistLimitDegrees=2}
 {bone1Name=dyn_SideSkirt_l_02, bone2Name=dyn_SideSkirt_l_01, swing1LimitDegrees=10, swing2LimitDegrees=30, twistLimitDegrees=2}
 {bone1Name=dyn_frontSkirt_l_02, bone2Name=dyn_frontSkirt_l_01, swing1LimitDegrees=10, swing2LimitDegrees=30, twistLimitDegrees=2}
 {bone1Name=dyn_frontSkirt_m_02, bone2Name=dyn_frontSkirt_m_01, swing1LimitDegrees=10, swing2LimitDegrees=30, twistLimitDegrees=2}
 {
  bone1Name=MSTR_BackSkirt_r
  bone2Name=pelvis
  swing1LimitDegrees=45
  swing1Motion=Locked
  swing2LimitDegrees=45
  swing2Motion=Locked
  twistLimitDegrees=45
  twistMotion=Locked
 }
 {bone1Name=dyn_BackSkirt_r_01, bone2Name=MSTR_BackSkirt_r, swing1LimitDegrees=10, swing2LimitDegrees=20, twistLimitDegrees=5}
 {bone1Name=dyn_BackSkirt_r_02, bone2Name=dyn_BackSkirt_r_01, swing1LimitDegrees=10, swing2LimitDegrees=20, twistLimitDegrees=5}
 {bone1Name=dyn_BackSkirt_r_03, bone2Name=dyn_BackSkirt_r_02, swing1LimitDegrees=10, swing2LimitDegrees=20, twistLimitDegrees=5}
 {
  bone1Name=MSTR_SideSkirt_r
  bone2Name=pelvis
  swing1LimitDegrees=45
  swing1Motion=Locked
  swing2LimitDegrees=45
  swing2Motion=Locked
  twistLimitDegrees=45
  twistMotion=Locked
 }
 {bone1Name=dyn_SideSkirt_r_01, bone2Name=MSTR_SideSkirt_r, swing1LimitDegrees=10, swing2LimitDegrees=20, twistLimitDegrees=5}
 {bone1Name=dyn_SideSkirt_r_02, bone2Name=dyn_SideSkirt_r_01, swing1LimitDegrees=10, swing2LimitDegrees=20, twistLimitDegrees=5}
 {bone1Name=dyn_SideSkirt_r_03, bone2Name=dyn_SideSkirt_r_02, swing1LimitDegrees=10, swing2LimitDegrees=20, twistLimitDegrees=5}
 {
  bone1Name=MSTR_frontSkirt_r
  bone2Name=pelvis
  swing1LimitDegrees=45
  swing1Motion=Locked
  swing2LimitDegrees=45
  swing2Motion=Locked
  twistLimitDegrees=45
  twistMotion=Locked
 }
 {bone1Name=dyn_frontSkirt_r_01, bone2Name=MSTR_frontSkirt_r, swing1LimitDegrees=10, swing2LimitDegrees=20, twistLimitDegrees=5}
 {bone1Name=dyn_frontSkirt_r_02, bone2Name=dyn_frontSkirt_r_01, swing1LimitDegrees=10, swing2LimitDegrees=20, twistLimitDegrees=5}
 {bone1Name=dyn_frontSkirt_r_03, bone2Name=dyn_frontSkirt_r_02, swing1LimitDegrees=10, swing2LimitDegrees=20, twistLimitDegrees=5}
]
