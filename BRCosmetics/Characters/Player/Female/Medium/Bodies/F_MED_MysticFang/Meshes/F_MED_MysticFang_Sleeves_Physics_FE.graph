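# Physics asset for the sleeves of the MysticFang medium female player body cosmetic, defining collision and joint limits for its dynamic sleeve bones.
# Each side has capsule bodies on an upper and a lower sleeve bone, plus a capsule on the forearm.
# Both sleeve bones constrain directly to their forearm rather than chaining to each other, with tight swing and twist limits that keep any sway small.
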PhysicsAsset /BRCosmetics/Characters/Player/Female/Medium/Bodies/F_MED_MysticFang/Meshes/F_MED_MysticFang_Sleeves_Physics_FE
bodies {
 dyn_slevee_dwn_l [{
   shapeType=Capsule
   length=8.505
   radius=3.005
   shapeName=dyn_slevee_dwn_l_capsule
   center=<7.3081 -0.2692 -5.6308>
   rotation=<rot -0.0003 -94.9997 -90.0001>
  }]
 dyn_slevee_dwn_r [{
   shapeType=Capsule
   length=8.505
   radius=3.005
   shapeName=dyn_slevee_dwn_r_capsule
   center=<7.3081 0.2692 5.3>
   rotation=<rot -0.0003 -94.9997 -90>
  }]
 dyn_slevee_uppr_l [{
   shapeType=Capsule
   length=3.63
   radius=3.005
   shapeName=dyn_slevee_dwn_l_capsule
   center=<4.2303 0 4.1742>
   rotation=<rot -0.0003 -94.9997 -90.0001>
  }]
 dyn_slevee_uppr_r [{shapeType=Capsule, length=3.63, radius=3.005, shapeName=dyn_slevee_dwn_r_capsule, center=<4.2303 0 -4>, rotation=<rot -0.0003 -94.9997 -90>}]
 lowerarm_l [{shapeType=Capsule, length=6, radius=4, shapeName=lowerarm_l_capsule, center=<8.6986 0.187 -0.0289>, rotation=<rot -0.0002 -87.9086 -90.5431>}]
 lowerarm_r [{shapeType=Capsule, length=6, radius=4, shapeName=lowerarm_r_capsule, center=<-8.6986 -0.187 0.0289>, rotation=<rot -0.0002 -87.9086 89.4569>}]
}
constraints [
 {bone1Name=dyn_slevee_dwn_l, bone2Name=lowerarm_l, swing1LimitDegrees=5, swing2LimitDegrees=5, twistLimitDegrees=1}
 {bone1Name=dyn_slevee_dwn_r, bone2Name=lowerarm_r, swing1LimitDegrees=5, swing2LimitDegrees=5, twistLimitDegrees=1}
 {bone1Name=dyn_slevee_uppr_l, bone2Name=lowerarm_l, swing1LimitDegrees=5, swing2LimitDegrees=5, twistLimitDegrees=1}
 {bone1Name=dyn_slevee_uppr_r, bone2Name=lowerarm_r, swing1LimitDegrees=5, swing2LimitDegrees=5, twistLimitDegrees=1}
]
