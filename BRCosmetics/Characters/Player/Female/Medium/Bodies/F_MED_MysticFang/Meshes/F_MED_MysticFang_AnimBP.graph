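# Post-process-style Anim Blueprint for the F_MED_MysticFang player body, adding secondary motion to skirt, sleeves, back cloth and bow.
# It copies the parent mesh pose into a Control Rig fed crouch, DBNO and slide states.
# Cached poses feed chained AnimDynamics springs, some with spherical limits, plus additive bone offsets.
# CustomIdle switches between two sleeve RigidBody setups before final sleeve springs and a SkyDive_Float_Flap-weighted back-cloth lift.
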AnimBlueprint /BRCosmetics/Characters/Player/Female/Medium/Bodies/F_MED_MysticFang/Meshes/F_MED_MysticFang_AnimBP
target_skeleton=&/F_MED_MysticFang_Skeleton
graphs [
 AnimGraph {nodes [
   Component To Local()
   RigidBody!(ComponentPose, Alpha, OverrideWorldGravity:Vector="0,0,0", ExternalForce) -> (Pose) {
    bOverrideWorldGravity=true
    overridePhysicsAsset=&/F_MED_MysticFang_Physics
    simulationSpace=BaseBoneSpace
    alphaBoolBlend {blendOption=Linear}
    baseBoneRef {boneName=pelvis}
    componentAppliedLinearAccClamp=<0 0 0>
    simSpaceSettings {maxAngularAcceleration=40, maxAngularVelocity=8, maxLinearAcceleration=3500, maxLinearVelocity=700, worldAlpha=1}
   }
   n1 = RigidBody(ComponentPose=Local To Component(LocalPose=Use cached pose 'CustomIdle'()), OverrideWorldGravity="0,0,-1250") {
    bOverrideWorldGravity=true
    overridePhysicsAsset=&/F_MED_MysticFang_Sleeves_Physics_FE
    simulationSpace=BaseBoneSpace
    alphaBoolBlend {blendOption=Linear}
    baseBoneRef {boneName=pelvis}
    overrideWorldGravity {z=-1250}
    simSpaceSettings {maxAngularAcceleration=40, maxAngularVelocity=8, maxLinearAcceleration=3500, maxLinearVelocity=700, velocityScaleZ=0, worldAlpha=1}
   }
   n2 = "Transform (Modify) Bone - Bone: dyn_slevee_dwn_l"(ComponentPose=@n1, Alpha=1, Translation="0,0,0", Rotation="-7.249,-177.032,3.331", Scale="1,1,1") {
    rotationMode=BMM_Additive
    translationMode=BMM_Additive
    alphaBoolBlend {blendOption=Linear}
    boneToModify {boneName=dyn_slevee_dwn_l}
    rotation=<rot -7.249 -177.032 3.331>
   }
   n3 = "Transform (Modify) Bone - Bone: dyn_slevee_uppr_l"(ComponentPose=@n2, Alpha=1, Translation="0,0,0", Rotation="46.278,-23.351,27.056", Scale="1,1,1") {
    rotationMode=BMM_Additive
    translationMode=BMM_Additive
    alphaBoolBlend {blendOption=Linear}
    boneToModify {boneName=dyn_slevee_uppr_l}
    rotation=<rot 46.278 -23.351 27.056>
   }
   n4 = "Transform (Modify) Bone - Bone: dyn_slevee_dwn_r"(ComponentPose=@n3, Alpha=1, Translation="0,0,0", Rotation="-21.144,152.162,40.597", Scale="1,1,1") {
    rotationMode=BMM_Additive
    translationMode=BMM_Additive
    alphaBoolBlend {blendOption=Linear}
    boneToModify {boneName=dyn_slevee_dwn_r}
    rotation=<rot -21.144 152.162 40.597>
   }
   n5 = "Transform (Modify) Bone - Bone: dyn_slevee_uppr_r"(ComponentPose=@n4, Alpha=1, Translation="0,0,0", Rotation="46.278,-23.351,27.056", Scale="1,1,1") {
    rotationMode=BMM_Additive
    translationMode=BMM_Additive
    alphaBoolBlend {blendOption=Linear}
    boneToModify {boneName=dyn_slevee_uppr_r}
    rotation=<rot 46.278 -23.351 27.056>
   }
   n6 = Component To Local(ComponentPose=@n5)
   n7 = RigidBody(ComponentPose=Local To Component(LocalPose=Use cached pose 'CustomIdle'()), OverrideWorldGravity="0,0,-1250") {
    bOverrideWorldGravity=true
    overridePhysicsAsset=&/F_MED_MysticFang_Sleeves_Physics
    simulationSpace=BaseBoneSpace
    alphaBoolBlend {blendOption=Linear}
    baseBoneRef {boneName=pelvis}
    overrideWorldGravity {z=-1250}
    simSpaceSettings {maxAngularAcceleration=40, maxAngularVelocity=8, maxLinearAcceleration=3500, maxLinearVelocity=700, velocityScaleZ=0, worldAlpha=1}
   }
   n8 = Component To Local(ComponentPose=@n7)
   n9 = Get CustomIdle() -> (CustomIdle:bool)
   n10 = Anim Dynamics - Bone: dyn_slevee_dwn_l!(
    ComponentPose=Local To Component(LocalPose=Blend Poses by bool(BlendPose_0=@n6, BlendPose_1=@n8, bActiveValue=@n9) {blendType=Linear})
    Alpha=1
    GravityOverride:Vector="0,0,-980"
    ExternalForce="0,0,0"
   ) -> (Pose) {
    angularDampingOverride=0.9999
    angularSpringConstant=10000
    bAngularSpring=true
    bOverrideAngularDamping=true
    bUseGravityOverride=true
    bUseSphericalLimits=true
    gravityScale=0.35
    simulationSpace=RootRelative
    alphaBoolBlend {blendOption=Linear}
    boundBone {boneName=dyn_slevee_dwn_l}
    componentAppliedLinearAccClamp=<2000 2000 2000>
    gravityOverride {z=-980}
    physicsBodyDefinitions [{
      boundBone {boneName=dyn_slevee_dwn_l}
      boxExtents=<20 30 30>
      constraintSetup {angularLimitsMax=<0 2 2>, angularLimitsMin=<0 -2 -2>, angularTarget=<1 0 0>}
      localJointOffset=<4 0 0>
     }]
    sphericalLimits [{limitRadius=0, drivingBone {boneName=None}, sphereLocalOffset=<0 0 0>}]
   }
   n11 = Anim Dynamics - Bone: dyn_slevee_uppr_l!(ComponentPose=@n10, Alpha=1, GravityOverride:Vector="0,0,-980", ExternalForce="0,0,0") -> (Pose) {
    angularDampingOverride=0.9999
    angularSpringConstant=10000
    bAngularSpring=true
    bOverrideAngularDamping=true
    bUseGravityOverride=true
    bUseSphericalLimits=true
    gravityScale=0.35
    simulationSpace=RootRelative
    alphaBoolBlend {blendOption=Linear}
    boundBone {boneName=dyn_slevee_uppr_l}
    componentAppliedLinearAccClamp=<2000 2000 2000>
    gravityOverride {z=-980}
    physicsBodyDefinitions [{
      boundBone {boneName=dyn_slevee_uppr_l}
      boxExtents=<20 30 30>
      constraintSetup {angularLimitsMax=<0 2 2>, angularLimitsMin=<0 -2 -2>, angularTarget=<1 0 0>}
      localJointOffset=<4 0 0>
     }]
    sphericalLimits [{limitRadius=0, drivingBone {boneName=None}, sphereLocalOffset=<0 0 0>}]
   }
   n12 = Anim Dynamics - Bone: dyn_slevee_dwn_r!(ComponentPose=@n11, Alpha=1, GravityOverride:Vector="0,0,-980", ExternalForce="0,0,0") -> (Pose) {
    angularDampingOverride=0.9999
    angularSpringConstant=10000
    bAngularSpring=true
    bOverrideAngularDamping=true
    bUseGravityOverride=true
    gravityScale=0.35
    simulationSpace=RootRelative
    alphaBoolBlend {blendOption=Linear}
    boundBone {boneName=dyn_slevee_dwn_r}
    componentAppliedLinearAccClamp=<2000 2000 2000>
    gravityOverride {z=-980}
    physicsBodyDefinitions [{
      boundBone {boneName=dyn_slevee_dwn_r}
      boxExtents=<20 30 30>
      constraintSetup {angularLimitsMax=<0 2 2>, angularLimitsMin=<0 -2 -2>, angularTarget=<1 0 0>}
      localJointOffset=<4 0 0>
     }]
   }
   n13 = Anim Dynamics - Bone: dyn_slevee_uppr_r!(ComponentPose=@n12, Alpha=1, GravityOverride:Vector="0,0,-980", ExternalForce="0,0,0") -> (Pose) {
    angularDampingOverride=0.9999
    angularSpringConstant=10000
    bAngularSpring=true
    bOverrideAngularDamping=true
    bUseGravityOverride=true
    gravityScale=0.35
    simulationSpace=RootRelative
    alphaBoolBlend {blendOption=Linear}
    boundBone {boneName=dyn_slevee_uppr_r}
    componentAppliedLinearAccClamp=<2000 2000 2000>
    gravityOverride {z=-980}
    physicsBodyDefinitions [{
      boundBone {boneName=dyn_slevee_uppr_r}
      boxExtents=<20 30 30>
      constraintSetup {angularLimitsMax=<0 2 2>, angularLimitsMin=<0 -2 -2>, angularTarget=<1 0 0>}
      localJointOffset=<4 0 0>
     }]
   }
   n14 = Get SkyDive_Float_Flap() -> (SkyDive_Float_Flap:double)
   n15 = "Transform (Modify) Bone - Bone: base_back_cloth_l"(ComponentPose=@n13, Alpha=@n14, Translation="0,0,2.5", Rotation="0,0,0", Scale="1,1,1") {
    rotationMode=BMM_Additive
    rotationSpace=BCS_BoneSpace
    translationMode=BMM_Additive
    translationSpace=BCS_BoneSpace
    alphaBoolBlend {blendOption=Linear}
    boneToModify {boneName=base_back_cloth_l}
    translation {z=2.5}
   }
   n16 = "Transform (Modify) Bone - Bone: base_back_cloth_r"(ComponentPose=@n15, Alpha=@n14, Translation="0,0,2.5", Rotation="0,0,0", Scale="1,1,1") {
    rotationMode=BMM_Additive
    rotationSpace=BCS_BoneSpace
    translationMode=BMM_Additive
    translationSpace=BCS_BoneSpace
    alphaBoolBlend {blendOption=Linear}
    boneToModify {boneName=base_back_cloth_r}
    translation {z=2.5}
   }
   Output Pose(Result=Component To Local(ComponentPose=@n16)) {layerGroup=None}
   n17 = Copy Pose From Mesh() {bCopyCurves=true, bUseAttachedParent=true}
   n18 = Control Rig!(
    Source=Component To Local(ComponentPose=Local To Component(LocalPose=@n17))
    Alpha
    Crouching:bool=Get bIsCrouching()
    DBNO:bool=Get bIsDBNO()
    KneeSliding:double=Get KneeSliding() -> (KneeSliding:double)
    BaseballSlide:double=Get BaseballSlide() -> (BaseballSlide:double)
    isCrouchStrafe:double=Get isCrouchStrafe float() -> (isCrouchStrafe float:double)
   ) -> (Pose) {
    alphaBoolBlend {blendOption=Linear}
    controlRigAssetReference {blueprintRigClass=&/F_MED_MysticFang_CtrlRig.F_MED_MysticFang_CtrlRig_C}
   }
   Save cached pose 'CRCache_Input'(Pose=@n18) {cacheName=CRCache_Input}
   n19 = Anim Dynamics - Bone: dyn_frontSkirt_m_01!(
    SimulationSpace:AnimPhysSimSpaceType Enum=BoneRelative
    ComponentPose=Local To Component(LocalPose=Use cached pose 'CRCache_Input'())
    Alpha=1
    GravityOverride:Vector="0,0,500"
   ) -> (Pose) {
    angularDampingOverride=0.9999
    angularSpringConstant=200
    bAngularSpring=true
    bOverrideAngularDamping=true
    bUseGravityOverride=true
    bUsePlanarLimit=false
    gravityScale=0.3
    simulationSpace=BoneRelative
    alphaBoolBlend {blendOption=Linear}
    alphaScaleBiasClamp {bInterpResult=true}
    boundBone {boneName=dyn_frontSkirt_m_01}
    componentAppliedLinearAccClamp=<1000 1000 1000>
    componentLinearAccScale=<1 1 1>
    componentLinearVelScale=<1 1 1>
    gravityOverride {z=500}
    physicsBodyDefinitions [{
      sphereCollisionRadius=5
      boundBone {boneName=dyn_frontSkirt_m_01}
      boxExtents=<40 40 40>
      constraintSetup {angularLimitsMax=<0 0 5>, angularLimitsMin=<0 -10 -5>, angularTarget=<0 0 0>}
      localJointOffset=<10 0 0>
     }]
    relativeSpaceBone {boneName=root}
   }
   n20 = Anim Dynamics - Bone: dyn_frontSkirt_m_02!(SimulationSpace:AnimPhysSimSpaceType Enum=BoneRelative, ComponentPose=@n19, Alpha=1, GravityOverride:Vector="0,0,500") -> (Pose) {
    angularDampingOverride=0.9999
    angularSpringConstant=200
    bAngularSpring=true
    bOverrideAngularDamping=true
    bUseGravityOverride=true
    bUsePlanarLimit=false
    gravityScale=0.3
    simulationSpace=BoneRelative
    alphaBoolBlend {blendOption=Linear}
    alphaScaleBiasClamp {bInterpResult=true}
    boundBone {boneName=dyn_frontSkirt_m_02}
    componentAppliedLinearAccClamp=<1000 1000 1000>
    componentLinearAccScale=<1 1 1>
    componentLinearVelScale=<1 1 1>
    gravityOverride {z=500}
    physicsBodyDefinitions [{
      sphereCollisionRadius=5
      boundBone {boneName=dyn_frontSkirt_m_02}
      boxExtents=<40 40 40>
      constraintSetup {angularLimitsMax=<0 0 5>, angularLimitsMin=<0 -20 -5>, angularTarget=<0 0 0>}
      localJointOffset=<10 0 0>
     }]
    relativeSpaceBone {boneName=root}
   }
   n21 = Anim Dynamics - Bone: dyn_frontSkirt_m_03!(SimulationSpace:AnimPhysSimSpaceType Enum=BoneRelative, ComponentPose=@n20, Alpha=1, GravityOverride:Vector="0,0,500") -> (Pose) {
    angularDampingOverride=0.9999
    angularSpringConstant=200
    bAngularSpring=true
    bOverrideAngularDamping=true
    bUseGravityOverride=true
    bUsePlanarLimit=false
    gravityScale=0.3
    simulationSpace=BoneRelative
    alphaBoolBlend {blendOption=Linear}
    alphaScaleBiasClamp {bInterpResult=true}
    boundBone {boneName=dyn_frontSkirt_m_03}
    componentAppliedLinearAccClamp=<1000 1000 1000>
    componentLinearAccScale=<1 1 1>
    componentLinearVelScale=<1 1 1>
    gravityOverride {z=500}
    physicsBodyDefinitions [{
      sphereCollisionRadius=5
      boundBone {boneName=dyn_frontSkirt_m_03}
      boxExtents=<40 40 40>
      constraintSetup {angularLimitsMax=<0 0 5>, angularLimitsMin=<0 -30 -5>, angularTarget=<0 0 0>}
      localJointOffset=<10 0 0>
     }]
    relativeSpaceBone {boneName=root}
   }
   n22 = Anim Dynamics - Bone: dyn_BackSkirt_m_01!(SimulationSpace:AnimPhysSimSpaceType Enum=BoneRelative, ComponentPose=@n21, Alpha=1, GravityOverride:Vector="0,0,-980") -> (Pose) {
    angularDampingOverride=0.9999
    angularSpringConstant=200
    bAngularSpring=true
    bOverrideAngularDamping=true
    bUseGravityOverride=true
    bUsePlanarLimit=false
    bUseSphericalLimits=true
    gravityScale=0.3
    simulationSpace=BoneRelative
    alphaBoolBlend {blendOption=Linear}
    alphaScaleBiasClamp {bInterpResult=true}
    boundBone {boneName=dyn_BackSkirt_m_01}
    componentAppliedLinearAccClamp=<1000 1000 1000>
    componentLinearAccScale=<1 1 1>
    componentLinearVelScale=<1 1 1>
    gravityOverride {z=-980}
    physicsBodyDefinitions [{
      sphereCollisionRadius=5
      boundBone {boneName=dyn_BackSkirt_m_01}
      boxExtents=<40 40 40>
      constraintSetup {angularLimitsMax=<0 0 5>, angularLimitsMin=<0 -10 -5>, angularTarget=<0 0 0>}
      localJointOffset=<10 0 0>
     }]
    relativeSpaceBone {boneName=root}
   }
   n23 = Anim Dynamics - Bone: dyn_BackSkirt_m_02!(SimulationSpace:AnimPhysSimSpaceType Enum=BoneRelative, ComponentPose=@n22, Alpha=1, GravityOverride:Vector="0,0,-980") -> (Pose) {
    angularDampingOverride=0.9999
    angularSpringConstant=200
    bAngularSpring=true
    bOverrideAngularDamping=true
    bUseGravityOverride=true
    bUsePlanarLimit=false
    bUseSphericalLimits=true
    gravityScale=0.3
    simulationSpace=BoneRelative
    alphaBoolBlend {blendOption=Linear}
    alphaScaleBiasClamp {bInterpResult=true}
    boundBone {boneName=dyn_BackSkirt_m_02}
    componentAppliedLinearAccClamp=<1000 1000 1000>
    componentLinearAccScale=<1 1 1>
    componentLinearVelScale=<1 1 1>
    gravityOverride {z=-980}
    physicsBodyDefinitions [{
      sphereCollisionRadius=5
      boundBone {boneName=dyn_BackSkirt_m_02}
      boxExtents=<40 40 40>
      constraintSetup {angularLimitsMax=<0 0 5>, angularLimitsMin=<0 -20 -5>, angularTarget=<0 0 0>}
      localJointOffset=<10 0 0>
     }]
    relativeSpaceBone {boneName=root}
   }
   n24 = Anim Dynamics - Bone: dyn_BackSkirt_m_03!(SimulationSpace:AnimPhysSimSpaceType Enum=BoneRelative, ComponentPose=@n23, Alpha=1, GravityOverride:Vector="0,0,-980") -> (Pose) {
    angularDampingOverride=0.9999
    angularSpringConstant=200
    bAngularSpring=true
    bOverrideAngularDamping=true
    bUseGravityOverride=true
    bUsePlanarLimit=false
    bUseSphericalLimits=true
    gravityScale=0.3
    simulationSpace=BoneRelative
    alphaBoolBlend {blendOption=Linear}
    alphaScaleBiasClamp {bInterpResult=true}
    boundBone {boneName=dyn_BackSkirt_m_03}
    componentAppliedLinearAccClamp=<1000 1000 1000>
    componentLinearAccScale=<1 1 1>
    componentLinearVelScale=<1 1 1>
    gravityOverride {z=-980}
    physicsBodyDefinitions [{
      sphereCollisionRadius=5
      boundBone {boneName=dyn_BackSkirt_m_03}
      boxExtents=<40 40 40>
      constraintSetup {angularLimitsMax=<0 0 5>, angularLimitsMin=<0 -30 -5>, angularTarget=<0 0 0>}
      localJointOffset=<10 0 0>
     }]
    relativeSpaceBone {boneName=root}
   }
   n25 = Anim Dynamics - Bone: dyn_frontSkirt_l_01!(SimulationSpace:AnimPhysSimSpaceType Enum=BoneRelative, ComponentPose=@n24, Alpha=1, GravityOverride:Vector="0,0,500") -> (Pose) {
    angularDampingOverride=0.9999
    angularSpringConstant=200
    bAngularSpring=true
    bOverrideAngularDamping=true
    bUseGravityOverride=true
    bUsePlanarLimit=false
    gravityScale=0.3
    simulationSpace=BoneRelative
    alphaBoolBlend {blendOption=Linear}
    alphaScaleBiasClamp {bInterpResult=true}
    boundBone {boneName=dyn_frontSkirt_l_01}
    componentAppliedLinearAccClamp=<1000 1000 1000>
    componentLinearAccScale=<1 1 1>
    componentLinearVelScale=<1 1 1>
    gravityOverride {z=500}
    physicsBodyDefinitions [{
      sphereCollisionRadius=5
      boundBone {boneName=dyn_frontSkirt_l_01}
      boxExtents=<40 40 40>
      constraintSetup {angularLimitsMax=<0 -2 4>, angularLimitsMin=<0 -10 -4>, angularTarget=<0 0 0>}
      localJointOffset=<10 0 0>
     }]
    relativeSpaceBone {boneName=root}
   }
   n26 = Anim Dynamics - Bone: dyn_frontSkirt_l_02!(SimulationSpace:AnimPhysSimSpaceType Enum=BoneRelative, ComponentPose=@n25, Alpha=1, GravityOverride:Vector="0,0,500") -> (Pose) {
    angularDampingOverride=0.9999
    angularSpringConstant=200
    bAngularSpring=true
    bOverrideAngularDamping=true
    bUseGravityOverride=true
    bUsePlanarLimit=false
    gravityScale=0.3
    simulationSpace=BoneRelative
    alphaBoolBlend {blendOption=Linear}
    alphaScaleBiasClamp {bInterpResult=true}
    boundBone {boneName=dyn_frontSkirt_l_02}
    componentAppliedLinearAccClamp=<1000 1000 1000>
    componentLinearAccScale=<1 1 1>
    componentLinearVelScale=<1 1 1>
    gravityOverride {z=500}
    physicsBodyDefinitions [{
      sphereCollisionRadius=5
      boundBone {boneName=dyn_frontSkirt_l_02}
      boxExtents=<40 40 40>
      constraintSetup {angularLimitsMax=<0 0 4>, angularLimitsMin=<0 -20 -4>, angularTarget=<0 0 0>}
      localJointOffset=<10 0 0>
     }]
    relativeSpaceBone {boneName=root}
   }
   n27 = Anim Dynamics - Bone: dyn_frontSkirt_l_03!(SimulationSpace:AnimPhysSimSpaceType Enum=BoneRelative, ComponentPose=@n26, Alpha=1, GravityOverride:Vector="0,0,500") -> (Pose) {
    angularDampingOverride=0.9999
    angularSpringConstant=200
    bAngularSpring=true
    bOverrideAngularDamping=true
    bUseGravityOverride=true
    bUsePlanarLimit=false
    bUseSphericalLimits=true
    gravityScale=0.3
    simulationSpace=BoneRelative
    alphaBoolBlend {blendOption=Linear}
    alphaScaleBiasClamp {bInterpResult=true}
    boundBone {boneName=dyn_frontSkirt_l_03}
    componentAppliedLinearAccClamp=<1000 1000 1000>
    componentLinearAccScale=<1 1 1>
    componentLinearVelScale=<1 1 1>
    gravityOverride {z=500}
    physicsBodyDefinitions [{
      boundBone {boneName=dyn_frontSkirt_l_03}
      boxExtents=<40 40 40>
      constraintSetup {angularLimitsMax=<0 0 4>, angularLimitsMin=<0 -30 -4>, angularTarget=<0 0 0>}
      localJointOffset=<10 0 0>
     }]
    relativeSpaceBone {boneName=root}
    sphericalLimits [{limitRadius=6, drivingBone {boneName=thigh_twist_01_l}, sphereLocalOffset=<0 3 0>}]
   }
   n28 = Anim Dynamics - Bone: dyn_SideSkirt_l_01!(SimulationSpace:AnimPhysSimSpaceType Enum=BoneRelative, ComponentPose=@n27, Alpha=1, GravityOverride:Vector="0,0,500") -> (Pose) {
    angularDampingOverride=0.9999
    angularSpringConstant=200
    bAngularSpring=true
    bOverrideAngularDamping=true
    bUseGravityOverride=true
    bUsePlanarLimit=false
    gravityScale=0.3
    simulationSpace=BoneRelative
    alphaBoolBlend {blendOption=Linear}
    alphaScaleBiasClamp {bInterpResult=true}
    boundBone {boneName=dyn_SideSkirt_l_01}
    componentAppliedLinearAccClamp=<1000 1000 1000>
    componentLinearAccScale=<1 1 1>
    componentLinearVelScale=<1 1 1>
    gravityOverride {z=500}
    physicsBodyDefinitions [{
      sphereCollisionRadius=5
      boundBone {boneName=dyn_SideSkirt_l_01}
      boxExtents=<40 40 40>
      constraintSetup {angularLimitsMax=<0 0 4>, angularLimitsMin=<0 -10 -4>, angularTarget=<0 0 0>}
      localJointOffset=<10 0 0>
     }]
    relativeSpaceBone {boneName=root}
   }
   n29 = Anim Dynamics - Bone: dyn_SideSkirt_l_02!(SimulationSpace:AnimPhysSimSpaceType Enum=BoneRelative, ComponentPose=@n28, Alpha=1, GravityOverride:Vector="0,0,500") -> (Pose) {
    angularDampingOverride=0.9999
    angularSpringConstant=200
    bAngularSpring=true
    bOverrideAngularDamping=true
    bUseGravityOverride=true
    bUsePlanarLimit=false
    gravityScale=0.3
    simulationSpace=BoneRelative
    alphaBoolBlend {blendOption=Linear}
    alphaScaleBiasClamp {bInterpResult=true}
    boundBone {boneName=dyn_SideSkirt_l_02}
    componentAppliedLinearAccClamp=<1000 1000 1000>
    componentLinearAccScale=<1 1 1>
    componentLinearVelScale=<1 1 1>
    gravityOverride {z=500}
    physicsBodyDefinitions [{
      sphereCollisionRadius=5
      boundBone {boneName=dyn_SideSkirt_l_02}
      boxExtents=<40 40 40>
      constraintSetup {angularLimitsMax=<0 0 4>, angularLimitsMin=<0 -20 -4>, angularTarget=<0 0 0>}
      localJointOffset=<10 0 0>
     }]
    relativeSpaceBone {boneName=root}
   }
   n30 = Anim Dynamics - Bone: dyn_SideSkirt_l_03!(SimulationSpace:AnimPhysSimSpaceType Enum=BoneRelative, ComponentPose=@n29, Alpha=1, GravityOverride:Vector="0,0,500") -> (Pose) {
    angularDampingOverride=0.9999
    angularSpringConstant=200
    bAngularSpring=true
    bOverrideAngularDamping=true
    bUseGravityOverride=true
    bUsePlanarLimit=false
    gravityScale=0.3
    simulationSpace=BoneRelative
    alphaBoolBlend {blendOption=Linear}
    alphaScaleBiasClamp {bInterpResult=true}
    boundBone {boneName=dyn_SideSkirt_l_03}
    componentAppliedLinearAccClamp=<1000 1000 1000>
    componentLinearAccScale=<1 1 1>
    componentLinearVelScale=<1 1 1>
    gravityOverride {z=500}
    physicsBodyDefinitions [{
      boundBone {boneName=dyn_SideSkirt_l_03}
      boxExtents=<40 40 40>
      constraintSetup {angularLimitsMax=<0 0 4>, angularLimitsMin=<0 -30 -4>, angularTarget=<0 0 0>}
      localJointOffset=<10 0 0>
     }]
    relativeSpaceBone {boneName=root}
   }
   n31 = Anim Dynamics - Bone: dyn_BackSkirt_l_01!(SimulationSpace:AnimPhysSimSpaceType Enum=BoneRelative, ComponentPose=@n30, Alpha=1, GravityOverride:Vector="0,0,500") -> (Pose) {
    angularDampingOverride=0.9999
    angularSpringConstant=200
    bAngularSpring=true
    bOverrideAngularDamping=true
    bUseGravityOverride=true
    bUsePlanarLimit=false
    bUseSphericalLimits=true
    gravityScale=0.3
    simulationSpace=BoneRelative
    alphaBoolBlend {blendOption=Linear}
    alphaScaleBiasClamp {bInterpResult=true}
    boundBone {boneName=dyn_BackSkirt_l_01}
    componentAppliedLinearAccClamp=<1000 1000 1000>
    componentLinearAccScale=<1 1 1>
    componentLinearVelScale=<1 1 1>
    gravityOverride {z=500}
    physicsBodyDefinitions [{
      sphereCollisionRadius=5
      boundBone {boneName=dyn_BackSkirt_l_01}
      boxExtents=<40 40 40>
      constraintSetup {angularLimitsMax=<0 0 4>, angularLimitsMin=<0 -10 -4>, angularTarget=<0 0 0>}
      localJointOffset=<10 0 0>
     }]
    relativeSpaceBone {boneName=root}
   }
   n32 = Anim Dynamics - Bone: dyn_BackSkirt_l_02!(SimulationSpace:AnimPhysSimSpaceType Enum=BoneRelative, ComponentPose=@n31, Alpha=1, GravityOverride:Vector="0,0,500") -> (Pose) {
    angularDampingOverride=0.9999
    angularSpringConstant=200
    bAngularSpring=true
    bOverrideAngularDamping=true
    bUseGravityOverride=true
    bUsePlanarLimit=false
    bUseSphericalLimits=true
    gravityScale=0.3
    simulationSpace=BoneRelative
    alphaBoolBlend {blendOption=Linear}
    alphaScaleBiasClamp {bInterpResult=true}
    boundBone {boneName=dyn_BackSkirt_l_02}
    componentAppliedLinearAccClamp=<1000 1000 1000>
    componentLinearAccScale=<1 1 1>
    componentLinearVelScale=<1 1 1>
    gravityOverride {z=500}
    physicsBodyDefinitions [{
      sphereCollisionRadius=5
      boundBone {boneName=dyn_BackSkirt_l_02}
      boxExtents=<40 40 40>
      constraintSetup {angularLimitsMax=<0 0 4>, angularLimitsMin=<0 -20 -4>, angularTarget=<0 0 0>}
      localJointOffset=<10 0 0>
     }]
    relativeSpaceBone {boneName=root}
   }
   n33 = Anim Dynamics - Bone: dyn_BackSkirt_l_03!(SimulationSpace:AnimPhysSimSpaceType Enum=BoneRelative, ComponentPose=@n32, Alpha=1, GravityOverride:Vector="0,0,500") -> (Pose) {
    angularDampingOverride=0.9999
    angularSpringConstant=200
    bAngularSpring=true
    bOverrideAngularDamping=true
    bUseGravityOverride=true
    bUsePlanarLimit=false
    bUseSphericalLimits=true
    gravityScale=0.3
    simulationSpace=BoneRelative
    alphaBoolBlend {blendOption=Linear}
    alphaScaleBiasClamp {bInterpResult=true}
    boundBone {boneName=dyn_BackSkirt_l_03}
    componentAppliedLinearAccClamp=<1000 1000 1000>
    componentLinearAccScale=<1 1 1>
    componentLinearVelScale=<1 1 1>
    gravityOverride {z=500}
    physicsBodyDefinitions [{
      sphereCollisionRadius=5
      boundBone {boneName=dyn_BackSkirt_l_03}
      boxExtents=<40 40 40>
      constraintSetup {angularLimitsMax=<0 0 4>, angularLimitsMin=<0 -30 -4>, angularTarget=<0 0 0>}
      localJointOffset=<10 0 0>
     }]
    relativeSpaceBone {boneName=root}
   }
   n34 = Anim Dynamics - Bone: dyn_frontSkirt_r_01!(SimulationSpace:AnimPhysSimSpaceType Enum=BoneRelative, ComponentPose=@n33, Alpha=1, GravityOverride:Vector="0,0,500") -> (Pose) {
    angularDampingOverride=0.9999
    angularSpringConstant=200
    bAngularSpring=true
    bOverrideAngularDamping=true
    bUseGravityOverride=true
    bUsePlanarLimit=false
    gravityScale=0.3
    simulationSpace=BoneRelative
    alphaBoolBlend {blendOption=Linear}
    boundBone {boneName=dyn_frontSkirt_r_01}
    componentAppliedLinearAccClamp=<1000 1000 1000>
    componentLinearAccScale=<1 1 1>
    componentLinearVelScale=<1 1 1>
    gravityOverride {z=500}
    physicsBodyDefinitions [{
      sphereCollisionRadius=5
      boundBone {boneName=dyn_frontSkirt_r_01}
      boxExtents=<40 40 40>
      constraintSetup {angularLimitsMax=<0 0 4>, angularLimitsMin=<0 -10 -4>, angularTarget=<0 0 0>}
      localJointOffset=<10 0 0>
     }]
    relativeSpaceBone {boneName=root}
   }
   n35 = Anim Dynamics - Bone: dyn_frontSkirt_r_02!(SimulationSpace:AnimPhysSimSpaceType Enum=BoneRelative, ComponentPose=@n34, Alpha=1, GravityOverride:Vector="0,0,500") -> (Pose) {
    angularDampingOverride=0.9999
    angularSpringConstant=200
    bAngularSpring=true
    bOverrideAngularDamping=true
    bUseGravityOverride=true
    bUsePlanarLimit=false
    gravityScale=0.3
    simulationSpace=BoneRelative
    alphaBoolBlend {blendOption=Linear}
    boundBone {boneName=dyn_frontSkirt_r_02}
    componentAppliedLinearAccClamp=<1000 1000 1000>
    componentLinearAccScale=<1 1 1>
    componentLinearVelScale=<1 1 1>
    gravityOverride {z=500}
    physicsBodyDefinitions [{
      sphereCollisionRadius=5
      boundBone {boneName=dyn_frontSkirt_r_02}
      boxExtents=<40 40 40>
      constraintSetup {angularLimitsMax=<0 0 4>, angularLimitsMin=<0 -20 -4>, angularTarget=<0 0 0>}
      localJointOffset=<10 0 0>
     }]
    relativeSpaceBone {boneName=root}
   }
   n36 = Anim Dynamics - Bone: dyn_frontSkirt_r_03!(SimulationSpace:AnimPhysSimSpaceType Enum=BoneRelative, ComponentPose=@n35, Alpha=1, GravityOverride:Vector="0,0,500") -> (Pose) {
    angularDampingOverride=0.9999
    angularSpringConstant=200
    bAngularSpring=true
    bOverrideAngularDamping=true
    bUseGravityOverride=true
    bUsePlanarLimit=false
    bUseSphericalLimits=true
    gravityScale=0.3
    simulationSpace=BoneRelative
    alphaBoolBlend {blendOption=Linear}
    boundBone {boneName=dyn_frontSkirt_r_03}
    componentAppliedLinearAccClamp=<1000 1000 1000>
    componentLinearAccScale=<1 1 1>
    componentLinearVelScale=<1 1 1>
    gravityOverride {z=500}
    physicsBodyDefinitions [{
      sphereCollisionRadius=5
      boundBone {boneName=dyn_frontSkirt_r_03}
      boxExtents=<40 40 40>
      constraintSetup {angularLimitsMax=<0 0 4>, angularLimitsMin=<0 -30 -4>, angularTarget=<0 0 0>}
      localJointOffset=<10 0 0>
     }]
    relativeSpaceBone {boneName=root}
   }
   n37 = Anim Dynamics - Bone: dyn_SideSkirt_r_01!(SimulationSpace:AnimPhysSimSpaceType Enum=BoneRelative, ComponentPose=@n36, Alpha=1, GravityOverride:Vector="0,0,500") -> (Pose) {
    angularDampingOverride=0.9999
    angularSpringConstant=200
    bAngularSpring=true
    bOverrideAngularDamping=true
    bUseGravityOverride=true
    bUsePlanarLimit=false
    gravityScale=0.3
    simulationSpace=BoneRelative
    alphaBoolBlend {blendOption=Linear}
    boundBone {boneName=dyn_SideSkirt_r_01}
    componentAppliedLinearAccClamp=<1000 1000 1000>
    componentLinearAccScale=<1 1 1>
    componentLinearVelScale=<1 1 1>
    gravityOverride {z=500}
    physicsBodyDefinitions [{
      sphereCollisionRadius=5
      boundBone {boneName=dyn_SideSkirt_r_01}
      boxExtents=<40 40 40>
      constraintSetup {angularLimitsMax=<0 0 4>, angularLimitsMin=<0 -10 -4>, angularTarget=<0 0 0>}
      localJointOffset=<10 0 0>
     }]
    relativeSpaceBone {boneName=root}
   }
   n38 = Anim Dynamics - Bone: dyn_SideSkirt_r_02!(SimulationSpace:AnimPhysSimSpaceType Enum=BoneRelative, ComponentPose=@n37, Alpha=1, GravityOverride:Vector="0,0,500") -> (Pose) {
    angularDampingOverride=0.9999
    angularSpringConstant=200
    bAngularSpring=true
    bOverrideAngularDamping=true
    bUseGravityOverride=true
    bUsePlanarLimit=false
    gravityScale=0.3
    simulationSpace=BoneRelative
    alphaBoolBlend {blendOption=Linear}
    boundBone {boneName=dyn_SideSkirt_r_02}
    componentAppliedLinearAccClamp=<1000 1000 1000>
    componentLinearAccScale=<1 1 1>
    componentLinearVelScale=<1 1 1>
    gravityOverride {z=500}
    physicsBodyDefinitions [{
      sphereCollisionRadius=5
      boundBone {boneName=dyn_SideSkirt_r_02}
      boxExtents=<40 40 40>
      constraintSetup {angularLimitsMax=<0 0 4>, angularLimitsMin=<0 -20 -4>, angularTarget=<0 0 0>}
      localJointOffset=<10 0 0>
     }]
    relativeSpaceBone {boneName=root}
   }
   n39 = Anim Dynamics - Bone: dyn_SideSkirt_r_03!(SimulationSpace:AnimPhysSimSpaceType Enum=BoneRelative, ComponentPose=@n38, Alpha=1, GravityOverride:Vector="0,0,500") -> (Pose) {
    angularDampingOverride=0.9999
    angularSpringConstant=200
    bAngularSpring=true
    bOverrideAngularDamping=true
    bUseGravityOverride=true
    bUsePlanarLimit=false
    gravityScale=0.3
    simulationSpace=BoneRelative
    alphaBoolBlend {blendOption=Linear}
    boundBone {boneName=dyn_SideSkirt_r_03}
    componentAppliedLinearAccClamp=<1000 1000 1000>
    componentLinearAccScale=<1 1 1>
    componentLinearVelScale=<1 1 1>
    gravityOverride {z=500}
    physicsBodyDefinitions [{
      sphereCollisionRadius=5
      boundBone {boneName=dyn_SideSkirt_r_03}
      boxExtents=<40 40 40>
      constraintSetup {angularLimitsMax=<0 0 4>, angularLimitsMin=<0 -30 -4>, angularTarget=<0 0 0>}
      localJointOffset=<10 0 0>
     }]
    relativeSpaceBone {boneName=root}
   }
   n40 = Anim Dynamics - Bone: dyn_BackSkirt_r_01!(SimulationSpace:AnimPhysSimSpaceType Enum=BoneRelative, ComponentPose=@n39, Alpha=1, GravityOverride:Vector="0,0,500") -> (Pose) {
    angularDampingOverride=0.9999
    angularSpringConstant=200
    bAngularSpring=true
    bOverrideAngularDamping=true
    bUseGravityOverride=true
    bUsePlanarLimit=false
    bUseSphericalLimits=true
    gravityScale=0.3
    simulationSpace=BoneRelative
    alphaBoolBlend {blendOption=Linear}
    boundBone {boneName=dyn_BackSkirt_r_01}
    componentAppliedLinearAccClamp=<1000 1000 1000>
    componentLinearAccScale=<1 1 1>
    componentLinearVelScale=<1 1 1>
    gravityOverride {z=500}
    physicsBodyDefinitions [{
      sphereCollisionRadius=5
      boundBone {boneName=dyn_BackSkirt_r_01}
      boxExtents=<40 40 40>
      constraintSetup {angularLimitsMax=<0 0 4>, angularLimitsMin=<0 -10 -4>, angularTarget=<0 0 0>}
      localJointOffset=<10 0 0>
     }]
    relativeSpaceBone {boneName=root}
   }
   n41 = Anim Dynamics - Bone: dyn_BackSkirt_r_02!(SimulationSpace:AnimPhysSimSpaceType Enum=BoneRelative, ComponentPose=@n40, Alpha=1, GravityOverride:Vector="0,0,500") -> (Pose) {
    angularDampingOverride=0.9999
    angularSpringConstant=200
    bAngularSpring=true
    bOverrideAngularDamping=true
    bUseGravityOverride=true
    bUsePlanarLimit=false
    bUseSphericalLimits=true
    gravityScale=0.3
    simulationSpace=BoneRelative
    alphaBoolBlend {blendOption=Linear}
    boundBone {boneName=dyn_BackSkirt_r_02}
    componentAppliedLinearAccClamp=<1000 1000 1000>
    componentLinearAccScale=<1 1 1>
    componentLinearVelScale=<1 1 1>
    gravityOverride {z=500}
    physicsBodyDefinitions [{
      sphereCollisionRadius=5
      boundBone {boneName=dyn_BackSkirt_r_02}
      boxExtents=<40 40 40>
      constraintSetup {angularLimitsMax=<0 0 4>, angularLimitsMin=<0 -20 -4>, angularTarget=<0 0 0>}
      localJointOffset=<10 0 0>
     }]
    relativeSpaceBone {boneName=root}
   }
   n42 = Anim Dynamics - Bone: dyn_BackSkirt_r_03!(SimulationSpace:AnimPhysSimSpaceType Enum=BoneRelative, ComponentPose=@n41, Alpha=1, GravityOverride:Vector="0,0,500") -> (Pose) {
    angularDampingOverride=0.9999
    angularSpringConstant=200
    bAngularSpring=true
    bOverrideAngularDamping=true
    bUseGravityOverride=true
    bUsePlanarLimit=false
    bUseSphericalLimits=true
    gravityScale=0.3
    simulationSpace=BoneRelative
    alphaBoolBlend {blendOption=Linear}
    boundBone {boneName=dyn_BackSkirt_r_03}
    componentAppliedLinearAccClamp=<1000 1000 1000>
    componentLinearAccScale=<1 1 1>
    componentLinearVelScale=<1 1 1>
    gravityOverride {z=500}
    physicsBodyDefinitions [{
      sphereCollisionRadius=5
      boundBone {boneName=dyn_BackSkirt_r_03}
      boxExtents=<40 40 40>
      constraintSetup {angularLimitsMax=<0 0 4>, angularLimitsMin=<0 -30 -4>, angularTarget=<0 0 0>}
      localJointOffset=<10 0 0>
     }]
    relativeSpaceBone {boneName=root}
   }
   Save cached pose 'CRCache_SkirtDyn'(Pose=Component To Local(ComponentPose=@n42)) {cacheName=CRCache_SkirtDyn}
   n43 = "Transform (Modify) Bone - Bone: dyn_frontSkirt_m_01"!(
    ComponentPose=Local To Component(LocalPose=Use cached pose 'CRCache_SkirtDyn'())
    bAlphaBoolEnabled:bool=True
    Translation="0,0,0"
    Rotation="-6.5,0,0"
    Scale="1,1,1"
   ) -> (Pose) {
    alphaInputType=Bool
    rotationMode=BMM_Additive
    rotationSpace=BCS_BoneSpace
    translationMode=BMM_Additive
    translationSpace=BCS_BoneSpace
    alphaBoolBlend {blendInTime=0.2, blendOption=Linear, blendOutTime=0.2}
    boneToModify {boneName=dyn_frontSkirt_m_01}
    rotation {pitch=-6.5}
   }
   n44 = "Transform (Modify) Bone - Bone: dyn_BackSkirt_m_01"!(ComponentPose=@n43, bAlphaBoolEnabled:bool=True, Translation="0,0,1", Rotation="15,0,0", Scale="1,1,1") -> (Pose) {
    alphaInputType=Bool
    rotationMode=BMM_Additive
    rotationSpace=BCS_BoneSpace
    translationMode=BMM_Additive
    translationSpace=BCS_BoneSpace
    alphaBoolBlend {blendInTime=0.2, blendOption=Linear, blendOutTime=0.2}
    boneToModify {boneName=dyn_BackSkirt_m_01}
    rotation {pitch=15}
    translation {z=1}
   }
   n45 = "Transform (Modify) Bone - Bone: dyn_frontSkirt_l_01"(ComponentPose=@n44, Alpha=1, Translation="0,0,0", Rotation="-6.209,15.688,-0.558", Scale="1,1,1") {
    rotationMode=BMM_Additive
    rotationSpace=BCS_BoneSpace
    translationMode=BMM_Additive
    translationSpace=BCS_BoneSpace
    alphaBoolBlend {blendOption=Linear}
    alphaScaleBiasClamp {bInterpResult=true}
    boneToModify {boneName=dyn_frontSkirt_l_01}
    rotation=<rot -6.209 15.688 -0.558>
   }
   n46 = Anim Dynamics - Bone: dyn_back_cloth_01_l!(ComponentPose=@n45, Alpha=1, GravityOverride:Vector="0,0,-1800", ExternalForce="0,0,0") -> (Pose) {
    angularDampingOverride=0.9999
    angularSpringConstant=8000
    bAngularSpring=true
    bGravityOverrideInSimSpace=true
    bOverrideAngularDamping=true
    bUseGravityOverride=true
    bUsePlanarLimit=false
    bUseSphericalLimits=true
    gravityScale=0.35
    simulationSpace=RootRelative
    alphaBoolBlend {blendOption=Linear}
    boundBone {boneName=dyn_back_cloth_01_l}
    componentAppliedLinearAccClamp=<2000 2000 2000>
    componentLinearAccScale=<1 1 1>
    componentLinearVelScale=<1 1 1>
    gravityOverride {z=-980}
    physicsBodyDefinitions [{
      sphereCollisionRadius=2
      boundBone {boneName=dyn_back_cloth_01_l}
      boxExtents=<40 40 40>
      constraintSetup {angularLimitsMax=<0 0 30>, angularLimitsMin=<0 -15 0>, angularTarget=<1 0 0>}
      localJointOffset=<6 0 0>
     }]
    sphericalLimits [{limitRadius=11.8, drivingBone {boneName=spine_05}, sphereLocalOffset=<-2 3 -12.3391>}]
   }
   n47 = Anim Dynamics - Bone: dyn_back_cloth_01_r!(ComponentPose=@n46, Alpha=1, GravityOverride:Vector="0,0,-980", ExternalForce="0,0,0") -> (Pose) {
    angularDampingOverride=0.9999
    angularSpringConstant=8000
    bAngularSpring=true
    bGravityOverrideInSimSpace=true
    bOverrideAngularDamping=true
    bUseGravityOverride=true
    bUsePlanarLimit=false
    bUseSphericalLimits=true
    gravityScale=0.35
    simulationSpace=RootRelative
    alphaBoolBlend {blendOption=Linear}
    boundBone {boneName=dyn_back_cloth_01_r}
    componentAppliedLinearAccClamp=<2000 2000 2000>
    componentLinearAccScale=<1 1 1>
    componentLinearVelScale=<1 1 1>
    gravityOverride {z=-980}
    physicsBodyDefinitions [{
      sphereCollisionRadius=2
      boundBone {boneName=dyn_back_cloth_01_r}
      boxExtents=<40 40 40>
      constraintSetup {angularLimitsMax=<0 0 0>, angularLimitsMin=<0 -15 -30>, angularTarget=<1 0 0>}
      localJointOffset=<6 0 0>
     }]
    sphericalLimits [{limitRadius=11.8, drivingBone {boneName=spine_05}, sphereLocalOffset=<-1.6033 2.8098 12.3391>}]
   }
   n48 = Anim Dynamics - Bone: dyn_back_cloth_02_l!(ComponentPose=@n47, Alpha=1, GravityOverride:Vector="0,0,-1800", ExternalForce="0,0,0") -> (Pose) {
    angularDampingOverride=0.9999
    angularSpringConstant=20000
    bAngularSpring=true
    bGravityOverrideInSimSpace=true
    bOverrideAngularDamping=true
    bUseGravityOverride=true
    bUsePlanarLimit=false
    bUseSphericalLimits=true
    gravityScale=0.35
    simulationSpace=RootRelative
    alphaBoolBlend {blendOption=Linear}
    boundBone {boneName=dyn_back_cloth_02_l}
    componentAppliedLinearAccClamp=<2000 2000 2000>
    componentLinearAccScale=<1 1 1>
    componentLinearVelScale=<1 1 1>
    gravityOverride {z=-980}
    physicsBodyDefinitions [{
      sphereCollisionRadius=2
      boundBone {boneName=dyn_back_cloth_02_l}
      boxExtents=<40 40 40>
      constraintSetup {angularLimitsMax=<0 0 10>, angularLimitsMin=<0 -10 -10>, angularTarget=<1 0 0>}
      localJointOffset=<6 0 0>
     }]
    sphericalLimits [{limitRadius=11.8, drivingBone {boneName=spine_05}, sphereLocalOffset=<-1.6033 2.8098 -12.3391>}]
   }
   n49 = Anim Dynamics - Bone: dyn_back_cloth_02_r!(ComponentPose=@n48, Alpha=1, GravityOverride:Vector="0,0,-980", ExternalForce="0,0,0") -> (Pose) {
    angularDampingOverride=0.9999
    angularSpringConstant=8000
    bAngularSpring=true
    bGravityOverrideInSimSpace=true
    bOverrideAngularDamping=true
    bUseGravityOverride=true
    bUsePlanarLimit=false
    bUseSphericalLimits=true
    gravityScale=0.35
    simulationSpace=RootRelative
    alphaBoolBlend {blendOption=Linear}
    boundBone {boneName=dyn_back_cloth_02_r}
    componentAppliedLinearAccClamp=<2000 2000 2000>
    componentLinearAccScale=<1 1 1>
    componentLinearVelScale=<1 1 1>
    gravityOverride {z=-980}
    physicsBodyDefinitions [{
      sphereCollisionRadius=2
      boundBone {boneName=dyn_back_cloth_02_r}
      boxExtents=<40 40 40>
      constraintSetup {angularLimitsMax=<0 0 0>, angularLimitsMin=<0 -10 -5>, angularTarget=<1 0 0>}
      localJointOffset=<6 0 0>
     }]
    sphericalLimits [{limitRadius=11.8, drivingBone {boneName=spine_05}, sphereLocalOffset=<-1.6033 2.8098 12.3391>}]
   }
   n50 = Anim Dynamics - Bone: dyn_bow_01_r!(ComponentPose=@n49, Alpha=1, GravityOverride:Vector="0,0,-980", ExternalForce="0,0,0") -> (Pose) {
    angularDampingOverride=0.9999
    angularSpringConstant=2000
    bAngularSpring=true
    bOverrideAngularDamping=true
    bUseGravityOverride=true
    bUsePlanarLimit=false
    bUseSphericalLimits=true
    gravityScale=0.35
    simulationSpace=RootRelative
    alphaBoolBlend {blendOption=Linear}
    boundBone {boneName=dyn_bow_01_r}
    componentAppliedLinearAccClamp=<3000 3000 3000>
    componentLinearAccScale=<0.5 0.5 0.5>
    componentLinearVelScale=<0.5 0.5 0.5>
    gravityOverride {z=-980}
    physicsBodyDefinitions [{
      boundBone {boneName=dyn_bow_01_r}
      boxExtents=<20 30 30>
      constraintSetup {angularLimitsMax=<0 0 10>, angularLimitsMin=<0 -10 -10>, angularTarget=<1 0 0>}
      localJointOffset=<6 0 0>
     }]
    sphericalLimits [{limitRadius=10.5, drivingBone {boneName=spine_05}, sphereLocalOffset=<0 0 0>}]
   }
   n51 = Anim Dynamics - Bone: dyn_bow_02_r!(ComponentPose=@n50, Alpha=1, GravityOverride:Vector="0,0,-980", ExternalForce="0,0,0") -> (Pose) {
    angularDampingOverride=0.9999
    angularSpringConstant=2000
    bAngularSpring=true
    bOverrideAngularDamping=true
    bUseGravityOverride=true
    bUsePlanarLimit=false
    bUseSphericalLimits=true
    gravityScale=0.35
    simulationSpace=RootRelative
    alphaBoolBlend {blendOption=Linear}
    boundBone {boneName=dyn_bow_02_r}
    componentAppliedLinearAccClamp=<3000 3000 3000>
    componentLinearAccScale=<0.5 0.5 0.5>
    componentLinearVelScale=<0.5 0.5 0.5>
    gravityOverride {z=-980}
    physicsBodyDefinitions [{
      boundBone {boneName=dyn_bow_02_r}
      boxExtents=<20 30 30>
      constraintSetup {angularLimitsMax=<0 0 10>, angularLimitsMin=<0 -15 -10>, angularTarget=<1 0 0>}
      localJointOffset=<6 0 0>
     }]
    sphericalLimits [{limitRadius=15, drivingBone {boneName=spine_04}, sphereLocalOffset=<0 1 0>}]
   }
   n52 = Anim Dynamics - Bone: dyn_bow_01_l!(ComponentPose=@n51, Alpha=1, GravityOverride:Vector="0,0,-980", ExternalForce="0,0,0") -> (Pose) {
    angularDampingOverride=0.9999
    angularSpringConstant=2000
    bAngularSpring=true
    bOverrideAngularDamping=true
    bUseGravityOverride=true
    bUsePlanarLimit=false
    bUseSphericalLimits=true
    gravityScale=0.35
    simulationSpace=RootRelative
    alphaBoolBlend {blendOption=Linear}
    boundBone {boneName=dyn_bow_01_l}
    componentAppliedLinearAccClamp=<3000 3000 3000>
    componentLinearAccScale=<0.5 0.5 0.5>
    componentLinearVelScale=<0.5 0.5 0.5>
    gravityOverride {z=-980}
    physicsBodyDefinitions [{
      boundBone {boneName=dyn_bow_01_l}
      boxExtents=<20 30 30>
      constraintSetup {angularLimitsMax=<0 0 10>, angularLimitsMin=<0 -10 -10>, angularTarget=<1 0 0>}
      localJointOffset=<6 0 0>
     }]
    sphericalLimits [{limitRadius=10.5, drivingBone {boneName=spine_05}, sphereLocalOffset=<0 0 0>}]
   }
   n53 = Anim Dynamics - Bone: dyn_bow_02_l!(ComponentPose=@n52, Alpha=1, GravityOverride:Vector="0,0,-980", ExternalForce="0,0,0") -> (Pose) {
    angularDampingOverride=0.9999
    angularSpringConstant=2000
    bAngularSpring=true
    bOverrideAngularDamping=true
    bUseGravityOverride=true
    bUsePlanarLimit=false
    bUseSphericalLimits=true
    gravityScale=0.35
    simulationSpace=RootRelative
    alphaBoolBlend {blendOption=Linear}
    boundBone {boneName=dyn_bow_02_l}
    componentAppliedLinearAccClamp=<3000 3000 3000>
    componentLinearAccScale=<0.5 0.5 0.5>
    componentLinearVelScale=<0.5 0.5 0.5>
    gravityOverride {z=-980}
    physicsBodyDefinitions [{
      boundBone {boneName=dyn_bow_02_l}
      boxExtents=<20 30 30>
      constraintSetup {angularLimitsMax=<0 0 10>, angularLimitsMin=<0 -15 -10>, angularTarget=<1 0 0>}
      localJointOffset=<6 0 0>
     }]
    sphericalLimits [{limitRadius=15, drivingBone {boneName=spine_04}, sphereLocalOffset=<0 1 0>}]
   }
   n54 = "Transform (Modify) Bone - Bone: dyn_back_cloth_01_l"(ComponentPose=@n53, Alpha=1, Translation="0,0,0", Rotation="0,-10,0", Scale="1,1,1") {
    rotationMode=BMM_Additive
    rotationSpace=BCS_BoneSpace
    translationMode=BMM_Additive
    translationSpace=BCS_BoneSpace
    alphaBoolBlend {blendOption=Linear}
    alphaScaleBiasClamp {bInterpResult=true}
    boneToModify {boneName=dyn_back_cloth_01_l}
    rotation {yaw=-10}
   }
   n55 = "Transform (Modify) Bone - Bone: dyn_back_cloth_02_l"(ComponentPose=@n54, Alpha=1, Translation="0,0,0", Rotation="0,-5,0", Scale="1,1,1") {
    rotationMode=BMM_Additive
    rotationSpace=BCS_BoneSpace
    translationMode=BMM_Additive
    translationSpace=BCS_BoneSpace
    alphaBoolBlend {blendOption=Linear}
    alphaScaleBiasClamp {bInterpResult=true}
    boneToModify {boneName=dyn_back_cloth_02_l}
    rotation {yaw=-5}
   }
   n56 = "Transform (Modify) Bone - Bone: dyn_back_cloth_01_r"(ComponentPose=@n55, Alpha=1, Translation="0,0,0", Rotation="0,5,0", Scale="1,1,1") {
    rotationMode=BMM_Additive
    rotationSpace=BCS_BoneSpace
    translationMode=BMM_Additive
    translationSpace=BCS_BoneSpace
    alphaBoolBlend {blendOption=Linear}
    alphaScaleBiasClamp {bInterpResult=true}
    boneToModify {boneName=dyn_back_cloth_01_r}
    rotation {yaw=5}
   }
   n57 = "Transform (Modify) Bone - Bone: dyn_back_cloth_02_r"(ComponentPose=@n56, Alpha=1, Translation="0,0,0", Rotation="0,5,0", Scale="1,1,1") {
    rotationMode=BMM_Additive
    rotationSpace=BCS_BoneSpace
    translationMode=BMM_Additive
    translationSpace=BCS_BoneSpace
    alphaBoolBlend {blendOption=Linear}
    alphaScaleBiasClamp {bInterpResult=true}
    boneToModify {boneName=dyn_back_cloth_02_r}
    rotation {yaw=5}
   }
   Save cached pose 'CustomIdle'(Pose=Component To Local(ComponentPose=@n57)) {cacheName=CustomIdle}
   Get RigidBody_off() -> (RigidBody_off:double)
  ]}
 EventGraph {nodes [
   Event OnParachuteOpened()
   > Switch on EGliderType(Selection=Get PlayerGliderType() -> (PlayerGliderType=HangGlider))
   > Set isHangGliding(isHangGliding:bool=true) -> (Output_Get:bool)
   Event OnBeginKneeSliding() {is=title K2Node_Event}
   > Set KneeSliding(KneeSliding:double=1) -> (Output_Get:double)
   Event OnEndKneeSliding() {is=title K2Node_Event}
   > Set KneeSliding(KneeSliding=0)
   Event OnBeginJumping()
   > Set Jumping(Jumping:bool=true) -> (Output_Get:bool)
   Event OnEndFalling()
   > Set Jumping(Jumping=false)
   Event OnCharacterCustomizationLoaded()
   > n1 = Sequence()
   > Set isWearingBackbling(isWearingBackbling=CheckForBackbling())
   n2 = Branch(execute=@n1.then_1, Condition=Get bIsInFrontEnd())
   > Set Sim Space Skirt(Sim Space Skirt:AnimPhysSimSpaceType Enum=World) -> (Output_Get:AnimPhysSimSpaceType Enum)
   > Set SkirtDynAlpha(SkirtDynAlpha:double=0.2) -> (Output_Get:double)
   Set Sim Space Skirt(execute=@n2.else)
   > Set SkirtDynAlpha(SkirtDynAlpha=1)
   > n3 = Branch(Condition=OR Boolean!(A=Get bIsSkydiveFloating(), B=Get bIsSkyDiving(), C:bool=Get bIsParachuteOpened(), D:bool=Get bIsCrouching()) -> (ReturnValue))
   > Set SkirtDynAlpha(SkirtDynAlpha=0.5)
   Set SkirtDynAlpha(execute=@n3.else, SkirtDynAlpha=1)
   n4 = Event BlueprintUpdateAnimation()
   > Set Delta Time X(Delta Time X=@n4.DeltaTimeX)
   > n5 = Sequence!(execute) -> (then_0, then_1, then_2:Exec, then_3:Exec, then_4:Exec, then_5:Exec, then_6:Exec, then_7:Exec, then_8:Exec)
   > n6 = "Gravity Override State Function (Parameter Asset) (Nativized)"(
    CharacterPartAnimInstance=Self-Reference()
    GameTimeInSeconds=GetGameTimeInSeconds()
    Parameters=&/M_MED_MysticFang_Parameters
   )
   > Set GravityOverride(GravityOverride=@n6)
   > Set RBAN Settings Out(RBAN Settings Out=@n6.OutRigidBodyAnimNodeParameters)
   n7 = Get bIsCrouching()
   > Set RigidBody_off(RigidBody_off:double="To Float (Boolean)"(InBool=NOT Boolean(A=@n7))) -> (Output_Get:double)
   n8 = Get bIsSkydiveFloating()
   > Set SkyDive_Float(SkyDive_Float:double="To Float (Boolean)"(InBool=@n8)) -> (Output_Get:double)
   n9 = OR Boolean(A=@n8, B=Get bIsSkyDivingFromBus() -> (bIsSkyDivingFromBus:bool), C=Get bIsSkydiveDiveMode(), D=Get bIsParachuteOpened())
   > Set SkyDive_Float_Off(SkyDive_Float_Off:double="To Float (Boolean)"(InBool=@n9)) -> (Output_Get:double)
   > n10 = Set SkyDive(SkyDive:bool=@n9) -> (Output_Get:bool)
   > Set SkyDive_Float_Flap(SkyDive_Float_Flap:double="To Float (Boolean)"(InBool=@n10)) -> (Output_Get:double)
   Reroute Node(InputPin=@n5.then_3)
   n11 = Branch(execute=@n5.then_6, Condition=Get bIsInFrontEnd())
   > Set Sim Space Skirt(Sim Space Skirt=World)
   > Set SkirtDynAlpha(SkirtDynAlpha=0.3)
   Set Sim Space Skirt(execute=@n11.else)
   > Set SkirtDynAlpha(SkirtDynAlpha=1)
   n12 = Get bIsSprinting()
   n13 = Get bIsFalling()
   n14 = Get bIsJumping()
   n15 = Branch(execute=@n5.then_1, Condition=AND Boolean(A=Get bIsTacticalSprinting(), B=OR Boolean!(A=@n13, B=@n14, C=@n12) -> (ReturnValue)))
   > Set IsTacticalandJump(IsTacticalandJump:double=1) -> (Output_Get:double)
   Set IsTacticalandJump(execute=@n15.else, IsTacticalandJump=0)
   n16 = GetAnimInstance(self=GetOwnerSkeletalMeshComponent())
   Is Valid(exec=@n5.then_8, InputObject=@n16)
   n17 = AND Boolean(
    A=Montage_IsPlaying(self=@n16, Montage=/BRCosmetics/Animation/Game/MainPlayer/Menu/BR/CustomSkins/_s20/LittleEgg/LittleEgg_Female_Idle_ALT_Montage)
    B=NOT Boolean(A=Get bIsPlayingEmote())
   )
   > Set CustomIdle(CustomIdle:bool=@n17) -> (Output_Get:bool)
   n18 = PlayerMovement(execute=@n5.then_2, AnimInstance=Self-Reference(), XDotProductCoefficient=100, YDotProductCoefficient=100, ZDotProductCoefficient=100)
   n19 = Get bIsSliding()
   n20 = Get bIsKneeSliding()
   n21 = Get bIsBaseballSliding()
   > Branch(Condition=AND Boolean!(A=Get bIsCrouching(), B=NOT Boolean(A=@n19), C:bool=NOT Boolean(A=@n21), D:bool=NOT Boolean(A=@n20)) -> (ReturnValue))
   n22 = BreakVector(InVec=MakeVector(X=@n18, Y=@n18.YMovement))
   > Set CrouchLeft(CrouchLeft:double="To Float (Boolean)"(InBool="float < float"(A=@n22, B=-50))) -> (Output_Get:double)
   > Set CrouchRight(CrouchRight:double="To Float (Boolean)"(InBool="float > float"(A=@n22, B=50))) -> (Output_Get:double)
   > Set CrouchForward(CrouchForward:double="To Float (Boolean)"(InBool="float > float"(A=@n22.Y, B=40))) -> (Output_Get:double)
   > Set CrouchBackward(CrouchBackward:double="To Float (Boolean)"(InBool="float < float"(A=@n22.Y, B=-40))) -> (Output_Get:double)
   n23 = PlayerMovement(execute=@n5.then_4, AnimInstance=Self-Reference(), XDotProductCoefficient=100, YDotProductCoefficient=100, ZDotProductCoefficient=100)
   n24 = BreakVector(InVec=MakeVector(X=@n23, Y=@n23.YMovement))
   > Set moveLeft bool(moveLeft bool:bool="float < float"(A=@n24, B=-50)) -> (Output_Get:bool)
   > Set moveRight bool(moveRight bool:bool="float > float"(A=@n24, B=50)) -> (Output_Get:bool)
   n25 = AND Boolean!(
    A=Get bIsCrouching()
    B=Get bIsMoving2D()
    C=OR Boolean!(A=Get moveLeft bool() -> (moveLeft bool:bool), B=Get moveRight bool() -> (moveRight bool:bool)) -> (ReturnValue)
   ) -> (ReturnValue)
   > Set isCrouchStrafe float(isCrouchStrafe float:double=FInterpTo(Current=Get isCrouchStrafe float(), Target="To Float (Boolean)"(InBool=@n25), DeltaTime=Get Delta Time X(), InterpSpeed=10)) -> (Output_Get:double)
   n26 = Branch(execute=@n5.then_5, Condition=Get bIsCrouching())
   > Set Alpha_Crouching(Alpha_Crouching:double=1) -> (Output_Get:double)
   Set Alpha_Crouching(execute=@n26.else, Alpha_Crouching=0)
   n27 = Branch(
    execute=@n5.then_7
    Condition=OR Boolean!(A=Get bIsSkyDiving(), B=Get bIsCrouching(), C:bool=Get bIsPlayingEmote(), D:bool=Get bIsJumping(), E:bool=Get bIsFalling()) -> (ReturnValue)
   )
   > Set isIdle(isIdle:double=0) -> (Output_Get:double)
   Set isIdle(execute=@n27.else, isIdle=1)
   Event OnParachuteClosed()
   > Set isHangGliding(isHangGliding=false)
   Event OnBeginBaseballSliding() {is=title K2Node_Event}
   > Set BaseballSlide(BaseballSlide:double=1) -> (Output_Get:double)
   Event OnEndBaseballSliding() {is=title K2Node_Event}
   > Set BaseballSlide(BaseballSlide=0)
   n28 = Branch(Condition=Get bIsCrouching())
   > Set SkirtDynAlpha(SkirtDynAlpha=0.5)
   Set SkirtDynAlpha(execute=@n28.else, SkirtDynAlpha=1)
   GetCurveValue(self=@n16, CurveName=LittleEgg_Female_Idle_ALT)
   NOT Boolean(A=@n9)
   Reroute Node(InputPin:Wildcard) -> (OutputPin:Wildcard)
   Set Is Playing Custom Idle Alpha(Is Playing Custom Idle Alpha:double=SelectFloat(A=1, B=0, bPickA=@n17)) -> (Output_Get:double)
  ]}
]
variables [
 Alpha_Crouching {type=number}
 AlphaSkydive {type=number}
 BaseballSlide {type=number}
 CrouchBackward {type=number}
 CrouchForward {type=number}
 CrouchLeft {type=number}
 CrouchRight {type=number}
 CustomIdle {type=boolean}
 Delta Time X {type=number}
 GravityOverride {type=Vector}
 Is Playing Custom Idle Alpha {type=number}
 isCrouchStrafe float {type=number}
 isHangGliding {type=boolean}
 isIdle {type=number}
 IsTacticalandJump {type=number}
 isWearingBackbling {type=boolean}
 Jumping {type=boolean}
 KneeSliding {type=number}
 moveLeft bool {type=boolean}
 moveRight bool {type=boolean}
 RBAN Settings Out {type=FortRigidBodyAnimNodeParameters}
 RigidBody_off {type=number}
 Sim Space Skirt {type=AnimPhysSimSpaceType}
 SkirtDynAlpha {type=number}
 SkyDive {type=boolean}
 SkyDive_Float {type=number}
 SkyDive_Float_Flap {type=number}
 SkyDive_Float_Off {type=number}
]
