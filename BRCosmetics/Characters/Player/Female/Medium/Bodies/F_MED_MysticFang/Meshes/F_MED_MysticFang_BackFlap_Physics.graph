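# Physics asset for the back flap of the MysticFang female medium body cosmetic, with capsule bodies for its skirt and back cloth bones.
# Front, side and back skirt chains run from the pelvis through tiny MSTR anchors into two larger capsules, with every angular axis locked.
# Thigh and spine_05 capsules rotate freely against the pelvis.
# Only the left-side, two-link back cloth chain hanging off spine_05 has tuned swing and twist limits.
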PhysicsAsset /BRCosmetics/Characters/Player/Female/Medium/Bodies/F_MED_MysticFang/Meshes/F_MED_MysticFang_BackFlap_Physics
bodies {
 MSTR_BackSkirt_l [{shapeType=Capsule, length=0.505, radius=0.505, shapeName=MSTR_BackSkirt_l_capsule, rotation=<rot 0 -90 0>}]
 MSTR_BackSkirt_m [{shapeType=Capsule, length=0.505, radius=0.505, shapeName=MSTR_BackSkirt_m_capsule, rotation=<rot 0 -90 0>}]
 MSTR_BackSkirt_r [{shapeType=Capsule, length=0.505, radius=0.505, shapeName=MSTR_BackSkirt_r_capsule, rotation=<rot 0 -90 0>}]
 MSTR_SideSkirt_l [{shapeType=Capsule, length=0.505, radius=0.505, shapeName=MSTR_SideSkirt_l_capsule, rotation=<rot 0 -90 0>}]
 MSTR_SideSkirt_r [{shapeType=Capsule, length=0.505, radius=0.505, shapeName=MSTR_SideSkirt_r_capsule, rotation=<rot 0 -90 0>}]
 MSTR_frontSkirt_l [{shapeType=Capsule, length=0.505, radius=0.505, shapeName=MSTR_frontSkirt_l_capsule, rotation=<rot 0 -90 0>}]
 MSTR_frontSkirt_m [{shapeType=Capsule, length=0.505, radius=0.505, shapeName=MSTR_frontSkirt_m_capsule, rotation=<rot 0 -90 0>}]
 MSTR_frontSkirt_r [{shapeType=Capsule, length=0.505, radius=0.505, shapeName=MSTR_frontSkirt_r_capsule, rotation=<rot 0 -90 0>}]
 dyn_BackSkirt_l_01 [{
   shapeType=Capsule
   length=13.1046
   radius=7.1474
   shapeName=dyn_BackSkirt_l_01_capsule
   center=<10.4348 1.0564 2.7304>
   rotation=<rot 0 -4.4945 -84.5936>
  }]
 dyn_BackSkirt_l_03 [{
   shapeType=Capsule
   length=9.2836
   radius=6.3231
   shapeName=dyn_BackSkirt_l_03_capsule
   center=<3.8892 2.1214 -2.5956>
   rotation=<rot 19.9707 171.1695 -86.7632>
  }]
 dyn_BackSkirt_m_01 [{
   shapeType=Capsule
   length=11.2289
   radius=7.2312
   shapeName=dyn_BackSkirt_m_01_capsule
   center=<10.0305 -0.3773 3.0834>
   rotation=<rot 0 178.5376 -86.5567>
  }]
 dyn_BackSkirt_m_03 [{
   shapeType=Capsule
   length=9.0908
   radius=4.3825
   shapeName=dyn_BackSkirt_m_03_capsule
   center=<6.7542 -0.1349 -3.6759>
   rotation=<rot 0 0.0342 -88.2309>
  }]
 dyn_BackSkirt_r_01 [{
   shapeType=Capsule
   length=13.1046
   radius=7.1474
   shapeName=dyn_BackSkirt_r_01_capsule
   center=<-10.4348 -1.0564 -2.7304>
   rotation=<rot 0 -4.4945 95.4064>
  }]
 dyn_BackSkirt_r_03 [{
   shapeType=Capsule
   length=9.2836
   radius=6.3231
   shapeName=dyn_BackSkirt_r_03_capsule
   center=<-3.8892 -2.1214 2.5956>
   rotation=<rot 19.9707 171.1695 93.2368>
  }]
 dyn_SideSkirt_l_01 [{
   shapeType=Capsule
   length=10.6775
   radius=6.5
   shapeName=dyn_SideSkirt_l_01_capsule
   center=<11.1274 1.1168 6.0041>
   rotation=<rot 0 10.6725 -82.7733>
  }]
 dyn_SideSkirt_l_03 [{
   shapeType=Capsule
   length=8.2895
   radius=5.3349
   shapeName=dyn_SideSkirt_l_03_capsule
   center=<9.8699 -0.9077 5.5345>
   rotation=<rot 24.9994 177.469 -89.5928>
  }]
 dyn_SideSkirt_r_01 [{
   shapeType=Capsule
   length=10.6775
   radius=6.5
   shapeName=dyn_SideSkirt_r_01_capsule
   center=<-11.1274 -1.1168 -6.0041>
   rotation=<rot 0 10.6725 97.2267>
  }]
 dyn_SideSkirt_r_03 [{
   shapeType=Capsule
   length=8.2895
   radius=5.3349
   shapeName=dyn_SideSkirt_r_03_capsule
   center=<-9.8699 0.9077 -5.5345>
   rotation=<rot 24.9994 177.469 90.4072>
  }]
 dyn_back_cloth_01_l [{
   shapeType=Capsule
   length=8.5697
   radius=2.5523
   shapeName=dyn_back_cloth_01_l_capsule
   center=<4.5073 -1.2532 0.5438>
   rotation=<rot 1.0141 -1.9258 -87.8242>
  }]
 dyn_back_cloth_02_l [{
   shapeType=Capsule
   length=4.875
   radius=2.88
   shapeName=dyn_back_cloth_02_l_capsule
   center=<8.9655 1.6371 0.0123>
   rotation=<rot -9.1499 -2.0434 -84.6643>
  }]
 dyn_frontSkirt_l_01 [{
   shapeType=Capsule
   length=7.9847
   radius=5.5203
   shapeName=dyn_frontSkirt_l_01_capsule
   center=<11.8192 -0.6499 6.8377>
   rotation=<rot 4.7656 -176.8162 94.7822>
  }]
 dyn_frontSkirt_l_03 [{
   shapeType=Capsule
   length=11.4815
   radius=4.7707
   shapeName=dyn_frontSkirt_l_03_capsule
   center=<7.2529 -2.458 6.2369>
   rotation=<rot 2.9596 -177.0353 -80.2592>
  }]
 dyn_frontSkirt_m_01 [{
   shapeType=Capsule
   length=11.7555
   radius=4.6041
   shapeName=dyn_frontSkirt_m_01_capsule
   center=<10.4268 -0.02 8.0915>
   rotation=<rot 2.5823 -179.4662 -89.7739>
  }]
 dyn_frontSkirt_m_03 [{
   shapeType=Capsule
   length=9.1656
   radius=5.3563
   shapeName=dyn_frontSkirt_m_03_capsule
   center=<7.1407 -1.4874 8.7031>
   rotation=<rot 0 -179.9261 -82.6433>
  }]
 dyn_frontSkirt_r_01 [{
   shapeType=Capsule
   length=7.9847
   radius=5.5203
   shapeName=dyn_frontSkirt_r_01_capsule
   center=<-11.8192 0.6499 -6.8377>
   rotation=<rot 4.7656 -176.8162 -85.2178>
  }]
 dyn_frontSkirt_r_03 [{
   shapeType=Capsule
   length=11.4815
   radius=4.7707
   shapeName=dyn_frontSkirt_r_03_capsule
   center=<-7.2529 2.458 -6.2369>
   rotation=<rot 2.9596 -177.0353 99.7408>
  }]
 pelvis [{shapeType=Capsule, length=14.25, radius=9, shapeName=pelvis_capsule, center=<-2.9969 -0.6851 0>}]
 spine_05 [{
   shapeType=Capsule
   length=18.0621
   radius=8.3103
   shapeName=spine_05_capsule
   center=<-4.4607 0.122 -0.3227>
   rotation=<rot -0.1879 -88.9455 -1.2606>
  }]
 thigh_l [{
   shapeType=Capsule
   length=29.4545
   radius=8.7955
   shapeName=thigh_l_capsule
   center=<-20.3533 0.5352 0.82>
   rotation=<rot 88.8689 2.1555 -0.3167>
  }]
 thigh_r [{
   shapeType=Capsule
   length=29.4545
   radius=8.7955
   shapeName=thigh_r_capsule
   center=<20.3533 -0.5352 -0.82>
   rotation=<rot 88.8689 2.1555 179.6833>
  }]
}
constraints [
 {
  bone1Name=thigh_l
  bone2Name=pelvis
  swing1LimitDegrees=45
  swing1Motion=Free
  swing2LimitDegrees=45
  swing2Motion=Free
  twistLimitDegrees=45
  twistMotion=Free
 }
 {
  bone1Name=thigh_r
  bone2Name=pelvis
  swing1LimitDegrees=45
  swing1Motion=Free
  swing2LimitDegrees=45
  swing2Motion=Free
  twistLimitDegrees=45
  twistMotion=Free
 }
 {
  bone1Name=MSTR_BackSkirt_l
  bone2Name=pelvis
  swing1LimitDegrees=45
  swing1Motion=Locked
  swing2LimitDegrees=45
  swing2Motion=Locked
  twistLimitDegrees=45
  twistMotion=Locked
 }
 {
  bone1Name=dyn_BackSkirt_l_01
  bone2Name=MSTR_BackSkirt_l
  swing1LimitDegrees=45
  swing1Motion=Locked
  swing2LimitDegrees=45
  swing2Motion=Locked
  twistLimitDegrees=45
  twistMotion=Locked
 }
 {
  bone1Name=dyn_BackSkirt_l_03
  bone2Name=dyn_BackSkirt_l_01
  swing1LimitDegrees=45
  swing1Motion=Locked
  swing2LimitDegrees=45
  swing2Motion=Locked
  twistLimitDegrees=45
  twistMotion=Locked
 }
 {
  bone1Name=MSTR_BackSkirt_m
  bone2Name=pelvis
  swing1LimitDegrees=45
  swing1Motion=Locked
  swing2LimitDegrees=45
  swing2Motion=Locked
  twistLimitDegrees=45
  twistMotion=Locked
 }
 {
  bone1Name=dyn_BackSkirt_m_01
  bone2Name=MSTR_BackSkirt_m
  swing1LimitDegrees=45
  swing1Motion=Locked
  swing2LimitDegrees=45
  swing2Motion=Locked
  twistLimitDegrees=45
  twistMotion=Locked
 }
 {
  bone1Name=dyn_BackSkirt_m_03
  bone2Name=dyn_BackSkirt_m_01
  swing1LimitDegrees=45
  swing1Motion=Locked
  swing2LimitDegrees=45
  swing2Motion=Locked
  twistLimitDegrees=45
  twistMotion=Locked
 }
 {
  bone1Name=MSTR_BackSkirt_r
  bone2Name=pelvis
  swing1LimitDegrees=45
  swing1Motion=Locked
  swing2LimitDegrees=45
  swing2Motion=Locked
  twistLimitDegrees=45
  twistMotion=Locked
 }
 {
  bone1Name=dyn_BackSkirt_r_01
  bone2Name=MSTR_BackSkirt_r
  swing1LimitDegrees=45
  swing1Motion=Locked
  swing2LimitDegrees=45
  swing2Motion=Locked
  twistLimitDegrees=45
  twistMotion=Locked
 }
 {
  bone1Name=dyn_BackSkirt_r_03
  bone2Name=dyn_BackSkirt_r_01
  swing1LimitDegrees=45
  swing1Motion=Locked
  swing2LimitDegrees=45
  swing2Motion=Locked
  twistLimitDegrees=45
  twistMotion=Locked
 }
 {
  bone1Name=MSTR_SideSkirt_l
  bone2Name=pelvis
  swing1LimitDegrees=45
  swing1Motion=Locked
  swing2LimitDegrees=45
  swing2Motion=Locked
  twistLimitDegrees=45
  twistMotion=Locked
 }
 {
  bone1Name=dyn_SideSkirt_l_01
  bone2Name=MSTR_SideSkirt_l
  swing1LimitDegrees=45
  swing1Motion=Locked
  swing2LimitDegrees=45
  swing2Motion=Locked
  twistLimitDegrees=45
  twistMotion=Locked
 }
 {
  bone1Name=dyn_SideSkirt_l_03
  bone2Name=dyn_SideSkirt_l_01
  swing1LimitDegrees=45
  swing1Motion=Locked
  swing2LimitDegrees=45
  swing2Motion=Locked
  twistLimitDegrees=45
  twistMotion=Locked
 }
 {
  bone1Name=MSTR_SideSkirt_r
  bone2Name=pelvis
  swing1LimitDegrees=45
  swing1Motion=Locked
  swing2LimitDegrees=45
  swing2Motion=Locked
  twistLimitDegrees=45
  twistMotion=Locked
 }
 {
  bone1Name=dyn_SideSkirt_r_01
  bone2Name=MSTR_SideSkirt_r
  swing1LimitDegrees=45
  swing1Motion=Locked
  swing2LimitDegrees=45
  swing2Motion=Locked
  twistLimitDegrees=45
  twistMotion=Locked
 }
 {
  bone1Name=dyn_SideSkirt_r_03
  bone2Name=dyn_SideSkirt_r_01
  swing1LimitDegrees=45
  swing1Motion=Locked
  swing2LimitDegrees=45
  swing2Motion=Locked
  twistLimitDegrees=45
  twistMotion=Locked
 }
 {
  bone1Name=MSTR_frontSkirt_l
  bone2Name=pelvis
  swing1LimitDegrees=45
  swing1Motion=Locked
  swing2LimitDegrees=45
  swing2Motion=Locked
  twistLimitDegrees=45
  twistMotion=Locked
 }
 {
  bone1Name=dyn_frontSkirt_l_01
  bone2Name=MSTR_frontSkirt_l
  swing1LimitDegrees=45
  swing1Motion=Locked
  swing2LimitDegrees=45
  swing2Motion=Locked
  twistLimitDegrees=45
  twistMotion=Locked
 }
 {
  bone1Name=dyn_frontSkirt_l_03
  bone2Name=dyn_frontSkirt_l_01
  swing1LimitDegrees=45
  swing1Motion=Locked
  swing2LimitDegrees=45
  swing2Motion=Locked
  twistLimitDegrees=45
  twistMotion=Locked
 }
 {
  bone1Name=MSTR_frontSkirt_m
  bone2Name=pelvis
  swing1LimitDegrees=45
  swing1Motion=Locked
  swing2LimitDegrees=45
  swing2Motion=Locked
  twistLimitDegrees=45
  twistMotion=Locked
 }
 {
  bone1Name=dyn_frontSkirt_m_01
  bone2Name=MSTR_frontSkirt_m
  swing1LimitDegrees=45
  swing1Motion=Locked
  swing2LimitDegrees=45
  swing2Motion=Locked
  twistLimitDegrees=45
  twistMotion=Locked
 }
 {
  bone1Name=dyn_frontSkirt_m_03
  bone2Name=dyn_frontSkirt_m_01
  swing1LimitDegrees=45
  swing1Motion=Locked
  swing2LimitDegrees=45
  swing2Motion=Locked
  twistLimitDegrees=45
  twistMotion=Locked
 }
 {
  bone1Name=MSTR_frontSkirt_r
  bone2Name=pelvis
  swing1LimitDegrees=45
  swing1Motion=Locked
  swing2LimitDegrees=45
  swing2Motion=Locked
  twistLimitDegrees=45
  twistMotion=Locked
 }
 {
  bone1Name=dyn_frontSkirt_r_01
  bone2Name=MSTR_frontSkirt_r
  swing1LimitDegrees=45
  swing1Motion=Locked
  swing2LimitDegrees=45
  swing2Motion=Locked
  twistLimitDegrees=45
  twistMotion=Locked
 }
 {
  bone1Name=dyn_frontSkirt_r_03
  bone2Name=dyn_frontSkirt_r_01
  swing1LimitDegrees=45
  swing1Motion=Locked
  swing2LimitDegrees=45
  swing2Motion=Locked
  twistLimitDegrees=45
  twistMotion=Locked
 }
 {
  bone1Name=spine_05
  bone2Name=pelvis
  swing1LimitDegrees=45
  swing1Motion=Free
  swing2LimitDegrees=45
  swing2Motion=Free
  twistLimitDegrees=45
  twistMotion=Free
 }
 {bone1Name=dyn_back_cloth_01_l, bone2Name=spine_05, swing1LimitDegrees=0, swing2LimitDegrees=20, twistLimitDegrees=30}
 {bone1Name=dyn_back_cloth_02_l, bone2Name=dyn_back_cloth_01_l, swing1LimitDegrees=0, swing2LimitDegrees=20, twistLimitDegrees=10}
]
