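# Post-process style AnimBlueprint for the MysticFang female FaceAcc mesh, animating hair and earrings.
# Poses copied from parent and body meshes are layered, passed through a Control Rig fed by movement and equipment states, then cached.
# A backpack flag switches between spine-constrained, offset back hair and a simulated hair pose, before final corrective bone offsets.
# Secondary motion combines RigidBody physics, AnimDynamics springs and state-weighted bone tweaks.
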AnimBlueprint /BRCosmetics/Characters/Player/Female/Medium/Bodies/F_MED_MysticFang/Meshes/Parts/F_MED_MysticFang_FaceAcc_AnimBP
target_skeleton=&/F_MED_MysticFang_FaceAcc_Skeleton
graphs [
 AnimGraph {nodes [
   n1 = "Constraint - dyn_Hair_Back_01 (PPPP)"!(
    ComponentPose=Local To Component(LocalPose=Use cached pose 'CRCache_HairPose'())
    Alpha=1
    ConstraintWeights_0=0
    ConstraintWeights_1:float=0
    ConstraintWeights_2:float=1
    ConstraintWeights_3:float=0
   ) -> (Pose) {
    alphaBoolBlend {blendOption=Linear}
    boneToModify {boneName=dyn_Hair_Back_01}
    constraintSetup [{targetBone {boneName=neck_02}}, {targetBone {boneName=head}}, {targetBone {boneName=spine_05}}, {targetBone {boneName=neck_01}}]
    constraintWeights [1,1,1,1]
   }
   n2 = "Constraint - dyn_Hair_Back_02 (PPPPP)"!(
    ComponentPose=@n1
    Alpha=1
    ConstraintWeights_0=0.8
    ConstraintWeights_1:float=0.1
    ConstraintWeights_2:float=0
    ConstraintWeights_3:float=0
    ConstraintWeights_4:float=0.1
   ) -> (Pose) {
    alphaBoolBlend {blendOption=Linear}
    boneToModify {boneName=dyn_Hair_Back_02}
    constraintSetup [
     {targetBone {boneName=spine_05}}
     {targetBone {boneName=neck_02}}
     {targetBone {boneName=neck_01}}
     {targetBone {boneName=spine_03}}
     {targetBone {boneName=head}}
    ]
    constraintWeights [1,1,1,1,1]
   }
   n3 = "Constraint - dyn_Hair_Back_03 (PPP)"!(ComponentPose=@n2, Alpha=1, ConstraintWeights_0=0.6702, ConstraintWeights_1:float=0.1668, ConstraintWeights_2:float=0.163) -> (Pose) {
    alphaBoolBlend {blendOption=Linear}
    boneToModify {boneName=dyn_Hair_Back_03}
    constraintSetup [{targetBone {boneName=spine_05}}, {targetBone {boneName=clavicle_r}}, {targetBone {boneName=clavicle_l}}]
    constraintWeights [1,1,1]
   }
   n4 = "Constraint - dyn_Hair_Back_04 (PPPPPPP)"!(
    ComponentPose=@n3
    Alpha=1
    ConstraintWeights_0=0.3363
    ConstraintWeights_1:float=0.3008
    ConstraintWeights_2:float=0.1776
    ConstraintWeights_3:float=0.0749
    ConstraintWeights_4:float=0.0493
    ConstraintWeights_5:float=0.0469
    ConstraintWeights_6:float=0.0142
   ) -> (Pose) {
    alphaBoolBlend {blendOption=Linear}
    boneToModify {boneName=dyn_Hair_Back_04}
    constraintSetup [
     {targetBone {boneName=spine_05}}
     {targetBone {boneName=spine_04}}
     {targetBone {boneName=clavicle_r}}
     {targetBone {boneName=clavicle_l}}
     {targetBone {boneName=spine_03}}
     {targetBone {boneName=spine_02}}
     {targetBone {boneName=spine_01}}
    ]
    constraintWeights [1,1,1,1,1,1,1]
   }
   n5 = "Constraint - dyn_Hair_Back_05 (PPPPPPPP)"!(
    ComponentPose=@n4
    Alpha=1
    ConstraintWeights_0=0.1
    ConstraintWeights_1:float=0.2
    ConstraintWeights_2:float=0.1
    ConstraintWeights_3:float=0.2
    ConstraintWeights_4:float=0
    ConstraintWeights_5:float=0.4
    ConstraintWeights_6:float=0
    ConstraintWeights_7:float=0
   ) -> (Pose) {
    alphaBoolBlend {blendOption=Linear}
    boneToModify {boneName=dyn_Hair_Back_05}
    constraintSetup [
     {targetBone {boneName=spine_04}}
     {targetBone {boneName=spine_02}}
     {targetBone {boneName=spine_05}}
     {targetBone {boneName=spine_03}}
     {targetBone {boneName=clavicle_l}}
     {targetBone {boneName=spine_01}}
     {targetBone {boneName=clavicle_r}}
     {targetBone {boneName=pelvis}}
    ]
    constraintWeights [1,1,1,1,1,1,1,1]
   }
   n6 = "Constraint - dyn_Hair_Back_06 (PPPPP)"!(
    ComponentPose=@n5
    Alpha=1
    ConstraintWeights_0=0
    ConstraintWeights_1:float=1
    ConstraintWeights_2:float=0
    ConstraintWeights_3:float=0
    ConstraintWeights_4:float=0
   ) -> (Pose) {
    alphaBoolBlend {blendOption=Linear}
    boneToModify {boneName=dyn_Hair_Back_06}
    constraintSetup [
     {targetBone {boneName=pelvis}}
     {targetBone {boneName=spine_02}}
     {targetBone {boneName=spine_01}}
     {targetBone {boneName=spine_03}}
     {targetBone {boneName=spine_04}}
    ]
    constraintWeights [1,1,1,1,1]
   }
   n7 = "Transform (Modify) Bone - Bone: dyn_Hair_Back_02"(ComponentPose=@n6, Alpha=1, Translation="0,0,0", Rotation="-4,0,0", Scale="1,1,0.5") {
    rotationMode=BMM_Additive
    rotationSpace=BCS_BoneSpace
    scaleMode=BMM_Additive
    scaleSpace=BCS_BoneSpace
    translationMode=BMM_Additive
    translationSpace=BCS_BoneSpace
    alphaBoolBlend {blendOption=Linear}
    boneToModify {boneName=dyn_Hair_Back_02}
    rotation {pitch=-4}
    scale {z=0.5}
   }
   n8 = "Transform (Modify) Bone - Bone: dyn_Hair_Back_03"(ComponentPose=@n7, Alpha=1, Translation="0,0,-10", Rotation="-29.182,0.174,0.413", Scale="1,1,1") {
    rotationMode=BMM_Additive
    rotationSpace=BCS_BoneSpace
    scaleMode=BMM_Additive
    scaleSpace=BCS_BoneSpace
    translationMode=BMM_Additive
    translationSpace=BCS_BoneSpace
    alphaBoolBlend {blendOption=Linear}
    boneToModify {boneName=dyn_Hair_Back_03}
    rotation=<rot -29.1821 0.174 0.413>
    translation {z=-10}
   }
   n9 = "Transform (Modify) Bone - Bone: dyn_Hair_Back_05"(ComponentPose=@n8, Alpha=1, Translation="-3.623,-0.029,-3.081", Rotation="7.614,-1.477,2.909", Scale="1,1,1") {
    rotationMode=BMM_Additive
    rotationSpace=BCS_BoneSpace
    translationMode=BMM_Additive
    translationSpace=BCS_BoneSpace
    alphaBoolBlend {blendOption=Linear}
    boneToModify {boneName=dyn_Hair_Back_05}
    rotation=<rot 7.614 -1.477 2.909>
    translation=<-3.623 -0.029 -3.081>
   }
   n10 = "Transform (Modify) Bone - Bone: dyn_Hair_Back_06"(ComponentPose=@n9, Alpha=1, Translation="0,0,0", Rotation="20,0,0", Scale="0.5,1,1") {
    rotationMode=BMM_Additive
    rotationSpace=BCS_BoneSpace
    scaleMode=BMM_Additive
    scaleSpace=BCS_BoneSpace
    translationMode=BMM_Additive
    translationSpace=BCS_BoneSpace
    alphaBoolBlend {blendOption=Linear}
    boneToModify {boneName=dyn_Hair_Back_06}
    rotation {pitch=20}
    scale {x=0.5}
   }
   n11 = "Transform (Modify) Bone - Bone: dyn_Hair_Back_02"(ComponentPose=@n10, Alpha=1, Translation="0,0,-4", Rotation="20,0,0", Scale="1,1,1") {
    rotationMode=BMM_Additive
    rotationSpace=BCS_BoneSpace
    translationMode=BMM_Additive
    translationSpace=BCS_BoneSpace
    alphaBoolBlend {blendOption=Linear}
    boneToModify {boneName=dyn_Hair_Back_02}
    rotation {pitch=20}
    translation {z=-4}
   }
   n12 = "Transform (Modify) Bone - Bone: dyn_hair_front_a_01_m"(ComponentPose=@n11, Alpha=1, Translation="0,1,0", Rotation="0,15,0", Scale="1,1,1") {
    rotationMode=BMM_Additive
    rotationSpace=BCS_BoneSpace
    scaleMode=BMM_Additive
    scaleSpace=BCS_BoneSpace
    translationMode=BMM_Additive
    translationSpace=BCS_BoneSpace
    alphaBoolBlend {blendOption=Linear}
    boneToModify {boneName=dyn_hair_front_a_01_m}
    rotation {yaw=15}
    translation {y=1}
   }
   n13 = "Transform (Modify) Bone - Bone: dyn_Hair_Back_01"(ComponentPose=@n12, Alpha=1, Translation="0,0,1", Rotation="0,0,0", Scale="1,1,1") {
    rotationMode=BMM_Additive
    rotationSpace=BCS_BoneSpace
    scaleMode=BMM_Additive
    scaleSpace=BCS_BoneSpace
    translationMode=BMM_Additive
    translationSpace=BCS_BoneSpace
    alphaBoolBlend {blendOption=Linear}
    boneToModify {boneName=dyn_Hair_Back_01}
    translation {z=1}
   }
   n14 = "Transform (Modify) Bone - Bone: dyn_Hair_Back_01"(ComponentPose=@n13, Alpha=1, Translation="8.550,-7.099,-1.535", Rotation="-8.635,-0.000,-0.000", Scale="1,1,1") {
    rotationMode=BMM_Additive
    rotationSpace=BCS_BoneSpace
    scaleMode=BMM_Additive
    scaleSpace=BCS_BoneSpace
    translationMode=BMM_Additive
    translationSpace=BCS_BoneSpace
    alphaBoolBlend {blendOption=Linear}
    alphaScaleBiasClamp {bInterpResult=true}
    boneToModify {boneName=dyn_Hair_Back_01}
    rotation=<rot -8.635 0 0>
    translation=<8.55 -7.099 -1.535>
   }
   n15 = "Transform (Modify) Bone - Bone: dyn_Hair_Back_03"(ComponentPose=@n14, Alpha=1, Translation="0.942,-0.564,-6.968", Rotation="6.848,3.635,8.252", Scale="1,1,1") {
    rotationMode=BMM_Additive
    rotationSpace=BCS_BoneSpace
    scaleMode=BMM_Additive
    scaleSpace=BCS_BoneSpace
    translationMode=BMM_Additive
    translationSpace=BCS_BoneSpace
    alphaBoolBlend {blendOption=Linear}
    alphaScaleBiasClamp {bInterpResult=true}
    boneToModify {boneName=dyn_Hair_Back_03}
    rotation=<rot 6.848 3.635 8.252>
    translation=<0.942 -0.564 -6.968>
   }
   n16 = "Transform (Modify) Bone - Bone: dyn_Hair_Back_05"(ComponentPose=@n15, Alpha=1, Translation="-7.683,0.185,7.244", Rotation="0,0,0", Scale="1,1,1") {
    rotationMode=BMM_Additive
    rotationSpace=BCS_BoneSpace
    scaleMode=BMM_Additive
    scaleSpace=BCS_BoneSpace
    translationMode=BMM_Additive
    translationSpace=BCS_BoneSpace
    alphaBoolBlend {blendOption=Linear}
    alphaScaleBiasClamp {bInterpResult=true}
    boneToModify {boneName=dyn_Hair_Back_05}
    translation=<-7.683 0.185 7.244>
   }
   n17 = "Transform (Modify) Bone - Bone: dyn_Hair_Back_03"(ComponentPose=@n16, Alpha=1, Translation="0,0,0", Rotation="8.844,0.863,0.324", Scale="1,1,1") {
    rotationMode=BMM_Additive
    rotationSpace=BCS_BoneSpace
    scaleMode=BMM_Additive
    scaleSpace=BCS_BoneSpace
    translationMode=BMM_Additive
    translationSpace=BCS_BoneSpace
    alphaBoolBlend {blendOption=Linear}
    alphaScaleBiasClamp {bInterpResult=true}
    boneToModify {boneName=dyn_Hair_Back_03}
    rotation=<rot 8.844 0.863 0.324>
   }
   n18 = "Transform (Modify) Bone - Bone: dyn_Hair_Back_02"(ComponentPose=@n17, Alpha=1, Translation="0,0,-8", Rotation="10,0.003,0.003", Scale="1,1,1") {
    rotationMode=BMM_Additive
    rotationSpace=BCS_BoneSpace
    scaleMode=BMM_Additive
    scaleSpace=BCS_BoneSpace
    translationMode=BMM_Additive
    translationSpace=BCS_BoneSpace
    alphaBoolBlend {blendOption=Linear}
    alphaScaleBiasClamp {bInterpResult=true}
    boneToModify {boneName=dyn_Hair_Back_02}
    rotation=<rot 10 0.003 0.003>
    translation {z=-8}
   }
   n19 = Anim Dynamics - Bone: dyn_hair_c_01_l!(ComponentPose=@n18, Alpha=1, GravityOverride:Vector="0,0,-980", ExternalForce="0,0,0") -> (Pose) {
    angularDampingOverride=0.9999
    angularSpringConstant=2000
    bAngularSpring=true
    bOverrideAngularDamping=true
    bUseGravityOverride=true
    gravityScale=0.35
    simulationSpace=RootRelative
    alphaBoolBlend {blendOption=Linear}
    boundBone {boneName=dyn_hair_c_01_l}
    componentAppliedLinearAccClamp=<0 0 0>
    gravityOverride {z=-980}
    physicsBodyDefinitions [{
      collisionType=CustomSphere
      sphereCollisionRadius=1
      boundBone {boneName=dyn_hair_c_01_l}
      boxExtents=<20 30 30>
      constraintSetup {angularLimitsMax=<0 0 0>, angularLimitsMin=<0 -15 0>, angularTarget=<1 0 0>}
      localJointOffset=<4 0 0>
     }]
   }
   n20 = Anim Dynamics - Bone: dyn_hair_a_02_l!(ComponentPose=@n19, Alpha=1, GravityOverride:Vector="0,0,-980", ExternalForce="0,0,0") -> (Pose) {
    angularDampingOverride=0.9999
    angularSpringConstant=2000
    bAngularSpring=true
    bOverrideAngularDamping=true
    bUseGravityOverride=true
    gravityScale=0.35
    simulationSpace=RootRelative
    alphaBoolBlend {blendOption=Linear}
    boundBone {boneName=dyn_hair_a_02_l}
    componentAppliedLinearAccClamp=<0 0 0>
    gravityOverride {z=-980}
    physicsBodyDefinitions [{
      collisionType=CustomSphere
      sphereCollisionRadius=1
      boundBone {boneName=dyn_hair_a_02_l}
      boxExtents=<20 30 30>
      constraintSetup {angularLimitsMax=<0 0 0>, angularLimitsMin=<0 -15 0>, angularTarget=<1 0 0>}
      localJointOffset=<4 0 0>
     }]
   }
   n21 = Anim Dynamics - Bone: dyn_hair_a_03_l!(ComponentPose=@n20, Alpha=1, GravityOverride:Vector="0,0,-980", ExternalForce="0,0,0") -> (Pose) {
    angularDampingOverride=0.9999
    angularSpringConstant=2000
    bAngularSpring=true
    bOverrideAngularDamping=true
    bUseGravityOverride=true
    gravityScale=0.35
    simulationSpace=RootRelative
    alphaBoolBlend {blendOption=Linear}
    boundBone {boneName=dyn_hair_a_03_l}
    componentAppliedLinearAccClamp=<0 0 0>
    gravityOverride {z=-980}
    physicsBodyDefinitions [{
      collisionType=CustomSphere
      sphereCollisionRadius=1
      boundBone {boneName=dyn_hair_a_03_l}
      boxExtents=<20 30 30>
      constraintSetup {angularLimitsMax=<0 0 0>, angularLimitsMin=<0 -15 0>, angularTarget=<1 0 0>}
      localJointOffset=<4 0 0>
     }]
   }
   n22 = Anim Dynamics - Bone: dyn_hair_e_01_l!(ComponentPose=@n21, Alpha=1, GravityOverride:Vector="0,0,-980", ExternalForce="0,0,0") -> (Pose) {
    angularDampingOverride=0.9999
    angularSpringConstant=2000
    bAngularSpring=true
    bOverrideAngularDamping=true
    bUseGravityOverride=true
    gravityScale=0.35
    simulationSpace=RootRelative
    alphaBoolBlend {blendOption=Linear}
    boundBone {boneName=dyn_hair_e_01_l}
    componentAppliedLinearAccClamp=<0 0 0>
    gravityOverride {z=-980}
    physicsBodyDefinitions [{
      collisionType=CustomSphere
      sphereCollisionRadius=1
      boundBone {boneName=dyn_hair_e_01_l}
      boxExtents=<20 30 30>
      constraintSetup {angularLimitsMax=<0 0 0>, angularLimitsMin=<0 -15 0>, angularTarget=<1 0 0>}
      localJointOffset=<4 0 0>
     }]
   }
   n23 = Anim Dynamics - Bone: dyn_hair_d_01_l!(ComponentPose=@n22, Alpha=1, GravityOverride:Vector="0,0,-980", ExternalForce="0,0,0") -> (Pose) {
    angularDampingOverride=0.9999
    angularSpringConstant=2000
    bAngularSpring=true
    bOverrideAngularDamping=true
    bUseGravityOverride=true
    gravityScale=0.35
    simulationSpace=RootRelative
    alphaBoolBlend {blendOption=Linear}
    boundBone {boneName=dyn_hair_d_01_l}
    componentAppliedLinearAccClamp=<0 0 0>
    gravityOverride {z=-980}
    physicsBodyDefinitions [{
      collisionType=CustomSphere
      sphereCollisionRadius=1
      boundBone {boneName=dyn_hair_d_01_l}
      boxExtents=<20 30 30>
      constraintSetup {angularLimitsMax=<0 0 0>, angularLimitsMin=<0 -15 0>, angularTarget=<1 0 0>}
      localJointOffset=<4 0 0>
     }]
   }
   n24 = Anim Dynamics - Bone: dyn_hair_b_01_l!(ComponentPose=@n23, Alpha=1, GravityOverride:Vector="0,0,-980", ExternalForce="0,0,0") -> (Pose) {
    angularDampingOverride=0.9999
    angularSpringConstant=2000
    bAngularSpring=true
    bOverrideAngularDamping=true
    bUseGravityOverride=true
    gravityScale=0.35
    simulationSpace=RootRelative
    alphaBoolBlend {blendOption=Linear}
    boundBone {boneName=dyn_hair_b_01_l}
    componentAppliedLinearAccClamp=<0 0 0>
    gravityOverride {z=-980}
    physicsBodyDefinitions [{
      collisionType=CustomSphere
      sphereCollisionRadius=1
      boundBone {boneName=dyn_hair_b_01_l}
      boxExtents=<20 30 30>
      constraintSetup {angularLimitsMax=<0 0 0>, angularLimitsMin=<0 -15 0>, angularTarget=<1 0 0>}
      localJointOffset=<4 0 0>
     }]
   }
   n25 = Anim Dynamics - Bone: dyn_hair_front_a_01_m!(ComponentPose=@n24, Alpha=0, GravityOverride:Vector="0,0,-980", ExternalForce="0,0,0") -> (Pose) {
    angularDampingOverride=0.9999
    angularSpringConstant=2000
    bAngularSpring=true
    bOverrideAngularDamping=true
    bUseGravityOverride=true
    gravityScale=0.35
    simulationSpace=RootRelative
    alphaBoolBlend {blendOption=Linear}
    boundBone {boneName=dyn_hair_front_a_01_m}
    componentLinearAccScale=<1 1 1>
    componentLinearVelScale=<1 1 1>
    gravityOverride {z=-980}
    physicsBodyDefinitions [{
      boundBone {boneName=dyn_hair_front_a_01_m}
      boxExtents=<20 30 30>
      constraintSetup {angularLimitsMax=<0 2 5>, angularLimitsMin=<0 -2 0>, angularTarget=<1 0 0>}
      localJointOffset=<4 0 0>
     }]
    planarLimits [{drivingBone {boneName=FACIAL_L_ForeheadIn}, planeTransform {location=<8.729 -1.9564 1.9128>, rotation=<rot 0 0 0>}}]
   }
   n26 = Anim Dynamics - Bone: dyn_hair_front_a_02_m!(ComponentPose=@n25, Alpha=0, GravityOverride:Vector="0,0,-980", ExternalForce="0,0,0") -> (Pose) {
    angularDampingOverride=0.9999
    angularSpringConstant=2000
    bAngularSpring=true
    bOverrideAngularDamping=true
    bUseGravityOverride=true
    bUseSphericalLimits=true
    gravityScale=0.35
    simulationSpace=RootRelative
    alphaBoolBlend {blendOption=Linear}
    boundBone {boneName=dyn_hair_front_a_02_m}
    componentAppliedLinearAccClamp=<0 0 0>
    gravityOverride {z=-980}
    physicsBodyDefinitions [{
      collisionType=CustomSphere
      sphereCollisionRadius=11
      boundBone {boneName=dyn_hair_front_a_02_m}
      boxExtents=<20 30 30>
      constraintSetup {angularLimitsMax=<0 2 0>, angularLimitsMin=<0 -2 0>, angularTarget=<1 0 0>}
      localJointOffset=<4 2.5393 6.3577>
     }]
    sphericalLimits [{limitRadius=0, drivingBone {boneName=head}, sphereLocalOffset=<0 0 0>}]
   }
   n27 = Anim Dynamics - Bone: dyn_ear_ring_l!(ComponentPose=@n26, Alpha=1, GravityOverride:Vector="0,0,-980", ExternalForce="0,0,0") -> (Pose) {
    angularDampingOverride=1
    angularSpringConstant=15000
    bAngularSpring=true
    bOverrideAngularDamping=true
    bUseGravityOverride=true
    bUseSphericalLimits=true
    gravityScale=0.35
    simulationSpace=RootRelative
    alphaBoolBlend {blendOption=Linear}
    boundBone {boneName=dyn_ear_ring_l}
    componentAppliedLinearAccClamp=<3000 3000 3000>
    gravityOverride {z=-980}
    physicsBodyDefinitions [{
      sphereCollisionRadius=2
      boundBone {boneName=dyn_ear_ring_l}
      boxExtents=<20 30 30>
      constraintSetup {angularLimitsMax=<0 15 15>, angularLimitsMin=<0 -15 -15>, angularTarget=<1 0 0>}
      localJointOffset=<5 0 0>
     }]
   }
   n28 = Anim Dynamics - Bone: dyn_ear_ring_r!(ComponentPose=@n27, Alpha=1, GravityOverride:Vector="0,0,-980", ExternalForce="0,0,0") -> (Pose) {
    angularDampingOverride=1
    angularSpringConstant=15000
    bAngularSpring=true
    bOverrideAngularDamping=true
    bUseGravityOverride=true
    bUseSphericalLimits=true
    gravityScale=0.35
    simulationSpace=RootRelative
    alphaBoolBlend {blendOption=Linear}
    boundBone {boneName=dyn_ear_ring_r}
    componentAppliedLinearAccClamp=<2000 2000 2000>
    componentLinearAccScale=<0.3 0.3 0.3>
    componentLinearVelScale=<0.3 0.3 0.3>
    gravityOverride {z=-980}
    physicsBodyDefinitions [{
      sphereCollisionRadius=2
      boundBone {boneName=dyn_ear_ring_r}
      boxExtents=<40 40 30>
      constraintSetup {angularLimitsMax=<0 5 40>, angularLimitsMin=<0 -20 -15>, angularTarget=<1 0 0>}
      localJointOffset=<15 0 0>
     }]
    sphericalLimits [{limitRadius=10, drivingBone {boneName=head}, sphereLocalOffset=<-15 0 0>}]
   }
   n29 = Component To Local(ComponentPose=@n28)
   n30 = "Transform (Modify) Bone - Bone: dyn_Hair_Back_02"(ComponentPose=Local To Component(LocalPose=Use cached pose 'HairDyn'()), Alpha=1, Translation="0,0,0", Rotation="-2,5,0", Scale="1,1,1") {
    rotationMode=BMM_Additive
    rotationSpace=BCS_BoneSpace
    translationMode=BMM_Additive
    translationSpace=BCS_BoneSpace
    alphaBoolBlend {blendOption=Linear}
    boneToModify {boneName=dyn_Hair_Back_02}
    rotation {pitch=-2, yaw=5}
   }
   n31 = Component To Local(ComponentPose=@n30)
   n32 = Get isWearingPackBool() -> (isWearingPackBool:bool)
   n33 = "Transform (Modify) Bone - Bone: dyn_Hair_Back_01"!(
    ComponentPose=Local To Component(LocalPose=Blend Poses by bool(BlendPose_0=@n29, BlendPose_1=@n31, bActiveValue=@n32) {blendType=Linear})
    bAlphaBoolEnabled:bool=True
    Translation="0,0,0"
    Rotation="-3,3,0"
    Scale="1,1,1"
   ) -> (Pose) {
    alphaInputType=Bool
    rotationMode=BMM_Additive
    rotationSpace=BCS_BoneSpace
    translationMode=BMM_Additive
    translationSpace=BCS_BoneSpace
    alphaBoolBlend {blendOption=Linear}
    boneToModify {boneName=dyn_Hair_Back_01}
    rotation {pitch=-3, yaw=3}
   }
   n34 = "Transform (Modify) Bone - Bone: dyn_Hair_Back_03"!(ComponentPose=@n33, bAlphaBoolEnabled:bool=True, Translation="0,0,0", Rotation="-2,0,0", Scale="1,1,1") -> (Pose) {
    alphaInputType=Bool
    rotationMode=BMM_Additive
    rotationSpace=BCS_BoneSpace
    translationMode=BMM_Additive
    translationSpace=BCS_BoneSpace
    alphaBoolBlend {blendOption=Linear}
    boneToModify {boneName=dyn_Hair_Back_03}
    rotation {pitch=-2}
   }
   n35 = "Transform (Modify) Bone - Bone: dyn_Hair_Back_04"!(ComponentPose=@n34, bAlphaBoolEnabled:bool=True, Translation="0,0,0", Rotation="0,5,0", Scale="1,1,1") -> (Pose) {
    alphaInputType=Bool
    rotationMode=BMM_Additive
    rotationSpace=BCS_BoneSpace
    translationMode=BMM_Additive
    translationSpace=BCS_BoneSpace
    alphaBoolBlend {blendOption=Linear}
    boneToModify {boneName=dyn_Hair_Back_04}
    rotation {yaw=5}
   }
   n36 = "Transform (Modify) Bone - Bone: dyn_Hair_Back_05"!(ComponentPose=@n35, bAlphaBoolEnabled:bool=True, Translation="0,0,0", Rotation="0,-10,0", Scale="1,1,1") -> (Pose) {
    alphaInputType=Bool
    rotationMode=BMM_Additive
    rotationSpace=BCS_BoneSpace
    translationMode=BMM_Additive
    translationSpace=BCS_BoneSpace
    alphaBoolBlend {blendOption=Linear}
    boneToModify {boneName=dyn_Hair_Back_05}
    rotation {yaw=-10}
   }
   n37 = "Transform (Modify) Bone - Bone: Hair_AltSkin_Corrective_01"(ComponentPose=@n36, Alpha=1, Translation="0,10,0", Rotation="0,0,-33.475", Scale="1,1,1") {
    rotationMode=BMM_Additive
    rotationSpace=BCS_BoneSpace
    translationMode=BMM_Additive
    translationSpace=BCS_BoneSpace
    boneToModify {boneName=Hair_AltSkin_Corrective_01}
    rotation {roll=-33.475}
    translation {y=10}
   }
   n38 = "Transform (Modify) Bone - Bone: Hair_AltSkin_Corrective_02"(ComponentPose=@n37, Alpha=1, Translation="0,20,0", Rotation="0,0,0", Scale="1,1,1") {
    rotationMode=BMM_Additive
    rotationSpace=BCS_BoneSpace
    scaleMode=BMM_Additive
    scaleSpace=BCS_BoneSpace
    translationMode=BMM_Additive
    translationSpace=BCS_BoneSpace
    boneToModify {boneName=Hair_AltSkin_Corrective_02}
    translation {y=20}
   }
   n39 = "Transform (Modify) Bone - Bone: Hair_AltSkin_Corrective_03"(ComponentPose=@n38, Alpha=1, Translation="0,21.066,5", Rotation="0,0,60", Scale="1,1,1") {
    rotationMode=BMM_Additive
    rotationSpace=BCS_BoneSpace
    scaleMode=BMM_Additive
    scaleSpace=BCS_BoneSpace
    translationMode=BMM_Additive
    translationSpace=BCS_BoneSpace
    boneToModify {boneName=Hair_AltSkin_Corrective_03}
    rotation {roll=60}
    translation {y=21.066, z=5}
   }
   Output Pose(Result=Component To Local(ComponentPose=@n39)) {layerGroup=None}
   n40 = Layered blend per bone(
    BasePose=Copy Pose From Mesh(SourceMeshComponent=Get Skel_Mesh()) {bUseAttachedParent=true}
    BlendPoses_0=Copy Pose From Mesh(SourceMeshComponent=Get Body_Mesh())
   ) {layerSetup [{branchFilters [{boneName=clavicle_l}, {boneName=clavicle_r}, {boneName=calf_l}, {boneName=calf_r}]}]}
   n41 = Control Rig!(
    Source=@n40
    Alpha
    isSkydiveFloating:double=Get isSkydiveFloating() -> (isSkydiveFloating:double)
    isHangGliding:double=Get isHangGliding() -> (isHangGliding:double)
    isCrouching:double=Get isCrouching() -> (isCrouching:double)
    isGunEquipped:double=Get isGunOut() -> (isGunOut:double)
    isTargeting:double=Get isTargeting() -> (isTargeting:double)
    isFrontEnd:bool=Get bIsInFrontEnd()
    isCapeEquipped:double=Get is_f_WearingCape() -> (is_f_WearingCape:double)
   ) -> (Pose) {
    alphaBoolBlend {blendOption=Linear}
    controlRigAssetReference {blueprintRigClass=&/F_MED_MysticFang_FaceAcc_CtrlRig.F_MED_MysticFang_FaceAcc_CtrlRig_C}
   }
   Save cached pose 'CRCache_HairPose'(Pose=@n41) {cacheName=CRCache_HairPose}
   n42 = RigidBody!(
    ComponentPose=Local To Component(LocalPose=Use cached pose 'CRCache_HairPose'())
    Alpha=Get Skydive_and_backpack_float_Not() -> (Skydive_and_backpack_float_Not:double)
    OverrideWorldGravity:Vector="0,0,0"
    ExternalForce
   ) -> (Pose) {
    bOverrideWorldGravity=true
    overridePhysicsAsset=&/F_MED_MysticFang_FaceAcc_Hair_Physics
    simulationSpace=BaseBoneSpace
    alphaBoolBlend {blendOption=Linear}
    baseBoneRef {boneName=spine_05}
    componentAppliedLinearAccClamp=<2000 2000 2000>
    simSpaceSettings {maxAngularAcceleration=40, maxAngularVelocity=8, maxLinearAcceleration=3500, maxLinearVelocity=700, velocityScaleZ=0.8, worldAlpha=0.8}
   }
   n43 = Get head_rot_back_Extreme() -> (head_rot_back_Extreme:double)
   n44 = "Transform (Modify) Bone - Bone: dyn_Hair_Back_01"!(ComponentPose=@n42, Alpha:float=@n43, Translation="0,0,0", Rotation="0,0,0", Scale="1,1,1") -> (Pose) {
    rotationMode=BMM_Additive
    rotationSpace=BCS_BoneSpace
    translationMode=BMM_Additive
    translationSpace=BCS_BoneSpace
    alphaBoolBlend {blendOption=Linear}
    boneToModify {boneName=dyn_Hair_Back_01}
   }
   n45 = "Transform (Modify) Bone - Bone: dyn_Hair_Back_02"(ComponentPose=@n44, Alpha=@n43, Translation="0,0,6", Rotation="-17,0,0", Scale="1,1,1") {
    rotationMode=BMM_Additive
    rotationSpace=BCS_BoneSpace
    translationMode=BMM_Additive
    translationSpace=BCS_BoneSpace
    alphaBoolBlend {blendOption=Linear}
    boneToModify {boneName=dyn_Hair_Back_02}
    rotation {pitch=-17}
    translation {z=6}
   }
   n46 = Get CrouchHeadsup() -> (CrouchHeadsup:double)
   n47 = "Transform (Modify) Bone - Bone: dyn_Hair_Back_04"!(ComponentPose=@n45, Alpha:float=@n46, Translation="0,0,0", Rotation="-5,0,0", Scale="1,1,1") -> (Pose) {
    rotationMode=BMM_Additive
    rotationSpace=BCS_BoneSpace
    translationMode=BMM_Additive
    translationSpace=BCS_BoneSpace
    alphaBoolBlend {blendOption=Linear}
    boneToModify {boneName=dyn_Hair_Back_04}
    rotation {pitch=-5}
   }
   n48 = Get CrouchWalkHeadsup() -> (CrouchWalkHeadsup:double)
   n49 = "Transform (Modify) Bone - Bone: dyn_Hair_Back_04"(ComponentPose=@n47, Alpha=@n48, Translation="0,0,0", Rotation="7,0,0", Scale="1,1,1") {
    rotationMode=BMM_Additive
    rotationSpace=BCS_BoneSpace
    translationMode=BMM_Additive
    translationSpace=BCS_BoneSpace
    alphaBoolBlend {blendOption=Linear}
    boneToModify {boneName=dyn_Hair_Back_04}
    rotation {pitch=7}
   }
   n50 = "Transform (Modify) Bone - Bone: dyn_Hair_Back_06"(ComponentPose=@n49, Alpha=@n48, Translation="0,0,6", Rotation="7,0,0", Scale="1,1,1") {
    rotationMode=BMM_Additive
    rotationSpace=BCS_BoneSpace
    translationMode=BMM_Additive
    translationSpace=BCS_BoneSpace
    alphaBoolBlend {blendOption=Linear}
    boneToModify {boneName=dyn_Hair_Back_06}
    rotation {pitch=7}
    translation {z=6}
   }
   n51 = "Transform (Modify) Bone - Bone: dyn_Hair_Back_06"(ComponentPose=@n50, Alpha=@n46, Translation="0,0,0", Rotation="0,0,0", Scale="1,1,1") {
    rotationMode=BMM_Additive
    rotationSpace=BCS_BoneSpace
    translationMode=BMM_Additive
    translationSpace=BCS_BoneSpace
    alphaBoolBlend {blendOption=Linear}
    boneToModify {boneName=dyn_Hair_Back_06}
   }
   n52 = Anim Dynamics - Bone: dyn_hair_c_01_l(ComponentPose=@n51, Alpha=1, GravityOverride="0,0,-980", ExternalForce="0,0,0") {
    angularDampingOverride=0.9999
    angularSpringConstant=2000
    bAngularSpring=true
    bOverrideAngularDamping=true
    bUseGravityOverride=true
    gravityScale=0.35
    simulationSpace=RootRelative
    alphaBoolBlend {blendOption=Linear}
    boundBone {boneName=dyn_hair_c_01_l}
    componentAppliedLinearAccClamp=<0 0 0>
    gravityOverride {z=-980}
    physicsBodyDefinitions [{
      collisionType=CustomSphere
      sphereCollisionRadius=1
      boundBone {boneName=dyn_hair_c_01_l}
      boxExtents=<20 30 30>
      constraintSetup {angularLimitsMax=<0 0 0>, angularLimitsMin=<0 -15 0>, angularTarget=<1 0 0>}
      localJointOffset=<4 0 0>
     }]
   }
   n53 = Anim Dynamics - Bone: dyn_hair_a_02_l(ComponentPose=@n52, Alpha=1, GravityOverride="0,0,-980", ExternalForce="0,0,0") {
    angularDampingOverride=0.9999
    angularSpringConstant=2000
    bAngularSpring=true
    bOverrideAngularDamping=true
    bUseGravityOverride=true
    gravityScale=0.35
    simulationSpace=RootRelative
    alphaBoolBlend {blendOption=Linear}
    boundBone {boneName=dyn_hair_a_02_l}
    componentAppliedLinearAccClamp=<0 0 0>
    gravityOverride {z=-980}
    physicsBodyDefinitions [{
      collisionType=CustomSphere
      sphereCollisionRadius=1
      boundBone {boneName=dyn_hair_a_02_l}
      boxExtents=<20 30 30>
      constraintSetup {angularLimitsMax=<0 0 0>, angularLimitsMin=<0 -15 0>, angularTarget=<1 0 0>}
      localJointOffset=<4 0 0>
     }]
   }
   n54 = Anim Dynamics - Bone: dyn_hair_a_03_l(ComponentPose=@n53, Alpha=1, GravityOverride="0,0,-980", ExternalForce="0,0,0") {
    angularDampingOverride=0.9999
    angularSpringConstant=2000
    bAngularSpring=true
    bOverrideAngularDamping=true
    bUseGravityOverride=true
    gravityScale=0.35
    simulationSpace=RootRelative
    alphaBoolBlend {blendOption=Linear}
    boundBone {boneName=dyn_hair_a_03_l}
    componentAppliedLinearAccClamp=<0 0 0>
    gravityOverride {z=-980}
    physicsBodyDefinitions [{
      collisionType=CustomSphere
      sphereCollisionRadius=1
      boundBone {boneName=dyn_hair_a_03_l}
      boxExtents=<20 30 30>
      constraintSetup {angularLimitsMax=<0 0 0>, angularLimitsMin=<0 -15 0>, angularTarget=<1 0 0>}
      localJointOffset=<4 0 0>
     }]
   }
   n55 = Anim Dynamics - Bone: dyn_hair_e_01_l(ComponentPose=@n54, Alpha=1, GravityOverride="0,0,-980", ExternalForce="0,0,0") {
    angularDampingOverride=0.9999
    angularSpringConstant=2000
    bAngularSpring=true
    bOverrideAngularDamping=true
    bUseGravityOverride=true
    gravityScale=0.35
    simulationSpace=RootRelative
    alphaBoolBlend {blendOption=Linear}
    boundBone {boneName=dyn_hair_e_01_l}
    componentAppliedLinearAccClamp=<0 0 0>
    gravityOverride {z=-980}
    physicsBodyDefinitions [{
      collisionType=CustomSphere
      sphereCollisionRadius=1
      boundBone {boneName=dyn_hair_e_01_l}
      boxExtents=<20 30 30>
      constraintSetup {angularLimitsMax=<0 0 0>, angularLimitsMin=<0 -15 0>, angularTarget=<1 0 0>}
      localJointOffset=<4 0 0>
     }]
   }
   n56 = Anim Dynamics - Bone: dyn_hair_d_01_l(ComponentPose=@n55, Alpha=1, GravityOverride="0,0,-980", ExternalForce="0,0,0") {
    angularDampingOverride=0.9999
    angularSpringConstant=2000
    bAngularSpring=true
    bOverrideAngularDamping=true
    bUseGravityOverride=true
    gravityScale=0.35
    simulationSpace=RootRelative
    alphaBoolBlend {blendOption=Linear}
    boundBone {boneName=dyn_hair_d_01_l}
    componentAppliedLinearAccClamp=<0 0 0>
    gravityOverride {z=-980}
    physicsBodyDefinitions [{
      collisionType=CustomSphere
      sphereCollisionRadius=1
      boundBone {boneName=dyn_hair_d_01_l}
      boxExtents=<20 30 30>
      constraintSetup {angularLimitsMax=<0 0 0>, angularLimitsMin=<0 -15 0>, angularTarget=<1 0 0>}
      localJointOffset=<4 0 0>
     }]
   }
   n57 = Anim Dynamics - Bone: dyn_hair_b_01_l(ComponentPose=@n56, Alpha=1, GravityOverride="0,0,-980", ExternalForce="0,0,0") {
    angularDampingOverride=0.9999
    angularSpringConstant=2000
    bAngularSpring=true
    bOverrideAngularDamping=true
    bUseGravityOverride=true
    gravityScale=0.35
    simulationSpace=RootRelative
    alphaBoolBlend {blendOption=Linear}
    boundBone {boneName=dyn_hair_b_01_l}
    componentAppliedLinearAccClamp=<0 0 0>
    gravityOverride {z=-980}
    physicsBodyDefinitions [{
      collisionType=CustomSphere
      sphereCollisionRadius=1
      boundBone {boneName=dyn_hair_b_01_l}
      boxExtents=<20 30 30>
      constraintSetup {angularLimitsMax=<0 0 0>, angularLimitsMin=<0 -15 0>, angularTarget=<1 0 0>}
      localJointOffset=<4 0 0>
     }]
   }
   n58 = Anim Dynamics - Bone: dyn_hair_front_a_01_m(ComponentPose=@n57, Alpha=0, GravityOverride="0,0,-980", ExternalForce="0,0,0") {
    angularDampingOverride=0.9999
    angularSpringConstant=2000
    bAngularSpring=true
    bOverrideAngularDamping=true
    bUseGravityOverride=true
    gravityScale=0.35
    simulationSpace=RootRelative
    alphaBoolBlend {blendOption=Linear}
    boundBone {boneName=dyn_hair_front_a_01_m}
    componentLinearAccScale=<1 1 1>
    componentLinearVelScale=<1 1 1>
    gravityOverride {z=-980}
    physicsBodyDefinitions [{
      boundBone {boneName=dyn_hair_front_a_01_m}
      boxExtents=<20 30 30>
      constraintSetup {angularLimitsMax=<0 2 5>, angularLimitsMin=<0 -2 0>, angularTarget=<1 0 0>}
      localJointOffset=<4 0 0>
     }]
    planarLimits [{drivingBone {boneName=FACIAL_L_ForeheadIn}, planeTransform {location=<8.729 -1.9564 1.9128>, rotation=<rot 0 0 0>}}]
   }
   n59 = Anim Dynamics - Bone: dyn_hair_front_a_02_m(ComponentPose=@n58, Alpha=0, GravityOverride="0,0,-980", ExternalForce="0,0,0") {
    angularDampingOverride=0.9999
    angularSpringConstant=2000
    bAngularSpring=true
    bOverrideAngularDamping=true
    bUseGravityOverride=true
    bUseSphericalLimits=true
    gravityScale=0.35
    simulationSpace=RootRelative
    alphaBoolBlend {blendOption=Linear}
    boundBone {boneName=dyn_hair_front_a_02_m}
    componentAppliedLinearAccClamp=<0 0 0>
    gravityOverride {z=-980}
    physicsBodyDefinitions [{
      collisionType=CustomSphere
      sphereCollisionRadius=11
      boundBone {boneName=dyn_hair_front_a_02_m}
      boxExtents=<20 30 30>
      constraintSetup {angularLimitsMax=<0 2 0>, angularLimitsMin=<0 -2 0>, angularTarget=<1 0 0>}
      localJointOffset=<4 2.5393 6.3577>
     }]
    sphericalLimits [{limitRadius=0, drivingBone {boneName=head}, sphereLocalOffset=<0 0 0>}]
   }
   n60 = Anim Dynamics - Bone: dyn_ear_ring_l(ComponentPose=@n59, Alpha=1, GravityOverride="0,0,-980", ExternalForce="0,0,0") {
    angularDampingOverride=1
    angularSpringConstant=15000
    bAngularSpring=true
    bOverrideAngularDamping=true
    bUseGravityOverride=true
    bUseSphericalLimits=true
    gravityScale=0.35
    simulationSpace=RootRelative
    alphaBoolBlend {blendOption=Linear}
    boundBone {boneName=dyn_ear_ring_l}
    componentAppliedLinearAccClamp=<3000 3000 3000>
    gravityOverride {z=-980}
    physicsBodyDefinitions [{
      sphereCollisionRadius=2
      boundBone {boneName=dyn_ear_ring_l}
      boxExtents=<20 30 30>
      constraintSetup {angularLimitsMax=<0 15 15>, angularLimitsMin=<0 -15 -15>, angularTarget=<1 0 0>}
      localJointOffset=<5 0 0>
     }]
   }
   n61 = Anim Dynamics - Bone: dyn_ear_ring_r(ComponentPose=@n60, Alpha=1, GravityOverride="0,0,-980", ExternalForce="0,0,0") {
    angularDampingOverride=1
    angularSpringConstant=15000
    bAngularSpring=true
    bOverrideAngularDamping=true
    bUseGravityOverride=true
    bUseSphericalLimits=true
    gravityScale=0.35
    simulationSpace=RootRelative
    alphaBoolBlend {blendOption=Linear}
    boundBone {boneName=dyn_ear_ring_r}
    componentAppliedLinearAccClamp=<2000 2000 2000>
    componentLinearAccScale=<0.3 0.3 0.3>
    componentLinearVelScale=<0.3 0.3 0.3>
    gravityOverride {z=-980}
    physicsBodyDefinitions [{
      sphereCollisionRadius=2
      boundBone {boneName=dyn_ear_ring_r}
      boxExtents=<40 40 30>
      constraintSetup {angularLimitsMax=<0 5 40>, angularLimitsMin=<0 -20 -15>, angularTarget=<1 0 0>}
      localJointOffset=<15 0 0>
     }]
    sphericalLimits [{limitRadius=10, drivingBone {boneName=head}, sphereLocalOffset=<-15 0 0>}]
   }
   n62 = "Transform (Modify) Bone - Bone: dyn_Hair_Back_01"(ComponentPose=@n61, Alpha=1, Translation="0,0,0", Rotation="0,-8,0", Scale="1,1,1") {
    rotationMode=BMM_Additive
    rotationSpace=BCS_BoneSpace
    translationMode=BMM_Additive
    translationSpace=BCS_BoneSpace
    alphaBoolBlend {blendOption=Linear}
    alphaScaleBiasClamp {bInterpResult=true}
    boneToModify {boneName=dyn_Hair_Back_01}
    rotation {yaw=-8}
   }
   n63 = "Transform (Modify) Bone - Bone: dyn_Hair_Back_03"!(ComponentPose=@n62, Alpha:float=1, Translation="0,0,0", Rotation="0,0,0", Scale="1,1,1") -> (Pose) {
    rotationMode=BMM_Additive
    rotationSpace=BCS_BoneSpace
    translationMode=BMM_Additive
    translationSpace=BCS_BoneSpace
    alphaBoolBlend {blendOption=Linear}
    alphaScaleBiasClamp {bInterpResult=true}
    boneToModify {boneName=dyn_Hair_Back_03}
   }
   n64 = "Transform (Modify) Bone - Bone: dyn_Hair_Back_05"!(ComponentPose=@n63, Alpha:float=1, Translation="0,0,0", Rotation="0,0,0", Scale="1,1,1") -> (Pose) {
    rotationMode=BMM_Additive
    rotationSpace=BCS_BoneSpace
    translationMode=BMM_Additive
    translationSpace=BCS_BoneSpace
    alphaBoolBlend {blendOption=Linear}
    alphaScaleBiasClamp {bInterpResult=true}
    boneToModify {boneName=dyn_Hair_Back_05}
   }
   n65 = "Constraint - dyn_Hair_Back_01 (P)"(ComponentPose=@n64, Alpha=1, ConstraintWeights_0=1) {
    alphaBoolBlend {blendOption=Linear}
    boneToModify {boneName=dyn_Hair_Back_01}
    constraintSetup [{targetBone {boneName=spine_05}}]
    constraintWeights [1]
   }
   n66 = "Constraint - dyn_Hair_Back_03 (P)"(ComponentPose=@n65, Alpha=1, ConstraintWeights_0=1) {
    alphaBoolBlend {blendOption=Linear}
    boneToModify {boneName=dyn_Hair_Back_03}
    constraintSetup [{targetBone {boneName=spine_05}}]
    constraintWeights [1]
   }
   n67 = Anim Dynamics - Bone: dyn_Hair_Back_03!(ComponentPose=@n66, Alpha=1, GravityOverride:Vector="0,0,-980", ExternalForce="0,0,0") -> (Pose) {
    angularDampingOverride=1
    angularSpringConstant=10000
    bAngularSpring=true
    bOverrideAngularDamping=true
    bUseGravityOverride=true
    gravityScale=0.35
    simulationSpace=RootRelative
    alphaBoolBlend {blendOption=Linear}
    alphaScaleBiasClamp {bInterpResult=true}
    boundBone {boneName=dyn_Hair_Back_03}
    componentAppliedLinearAccClamp=<0 0 0>
    gravityOverride {z=-980}
    physicsBodyDefinitions [{
      boundBone {boneName=dyn_Hair_Back_03}
      boxExtents=<20 30 30>
      constraintSetup {angularLimitsMax=<0 5 5>, angularLimitsMin=<0 -5 -5>, angularTarget=<1 0 0>}
      localJointOffset=<4 0 0>
     }]
   }
   n68 = Anim Dynamics - Bone: dyn_Hair_Back_04!(ComponentPose=@n67, Alpha=1, GravityOverride:Vector="0,0,-980", ExternalForce="0,0,0") -> (Pose) {
    angularDampingOverride=1
    angularSpringConstant=10000
    bAngularSpring=true
    bOverrideAngularDamping=true
    bUseGravityOverride=true
    gravityScale=0.35
    simulationSpace=RootRelative
    alphaBoolBlend {blendOption=Linear}
    alphaScaleBiasClamp {bInterpResult=true}
    boundBone {boneName=dyn_Hair_Back_04}
    componentAppliedLinearAccClamp=<0 0 0>
    gravityOverride {z=-980}
    physicsBodyDefinitions [{
      boundBone {boneName=dyn_Hair_Back_04}
      boxExtents=<20 30 30>
      constraintSetup {angularLimitsMax=<0 5 5>, angularLimitsMin=<0 -5 -5>, angularTarget=<1 0 0>}
      localJointOffset=<4 0 0>
     }]
   }
   n69 = Anim Dynamics - Bone: dyn_Hair_Back_05!(ComponentPose=@n68, Alpha=1, GravityOverride:Vector="0,0,-980", ExternalForce="0,0,0") -> (Pose) {
    angularDampingOverride=1
    angularSpringConstant=10000
    bAngularSpring=true
    bOverrideAngularDamping=true
    bUseGravityOverride=true
    gravityScale=0.35
    simulationSpace=RootRelative
    alphaBoolBlend {blendOption=Linear}
    alphaScaleBiasClamp {bInterpResult=true}
    boundBone {boneName=dyn_Hair_Back_05}
    componentAppliedLinearAccClamp=<0 0 0>
    gravityOverride {z=-980}
    physicsBodyDefinitions [{
      boundBone {boneName=dyn_Hair_Back_05}
      boxExtents=<20 30 30>
      constraintSetup {angularLimitsMax=<0 5 5>, angularLimitsMin=<0 -5 -5>, angularTarget=<1 0 0>}
      localJointOffset=<4 0 0>
     }]
   }
   n70 = Anim Dynamics - Bone: dyn_Hair_Back_06!(ComponentPose=@n69, Alpha=1, GravityOverride:Vector="0,0,-980", ExternalForce="0,0,0") -> (Pose) {
    angularDampingOverride=1
    angularSpringConstant=10000
    bAngularSpring=true
    bOverrideAngularDamping=true
    bUseGravityOverride=true
    bUsePlanarLimit=false
    gravityScale=0.35
    simulationSpace=RootRelative
    alphaBoolBlend {blendOption=Linear}
    alphaScaleBiasClamp {bInterpResult=true}
    boundBone {boneName=dyn_Hair_Back_06}
    componentAppliedLinearAccClamp=<0 0 0>
    gravityOverride {z=-980}
    physicsBodyDefinitions [{
      boundBone {boneName=dyn_Hair_Back_06}
      boxExtents=<20 30 30>
      constraintSetup {angularLimitsMax=<0 5 5>, angularLimitsMin=<0 -5 -5>, angularTarget=<1 0 0>}
      localJointOffset=<4 0 0>
     }]
    planarLimits [{drivingBone {boneName=pelvis}, planeTransform {location=<-4.2469 -10.8422 0>, rotation=<rot 0 -0.0001 -90>}}]
   }
   n71 = Anim Dynamics - Bone: dyn_Hair_Back_02!(ComponentPose=@n70, Alpha=0, GravityOverride:Vector="0,0,-980", ExternalForce="0,0,0") -> (Pose) {
    angularDampingOverride=0.9999
    angularSpringConstant=2000
    bAngularSpring=true
    bOverrideAngularDamping=true
    bUseGravityOverride=true
    gravityScale=0.35
    simulationSpace=RootRelative
    alphaBoolBlend {blendOption=Linear}
    boundBone {boneName=dyn_Hair_Back_02}
    componentAppliedLinearAccClamp=<3000 3000 3000>
    componentLinearAccScale=<1 1 1>
    componentLinearVelScale=<1 1 1>
    gravityOverride {z=-980}
    physicsBodyDefinitions [{
      boundBone {boneName=dyn_Hair_Back_02}
      boxExtents=<20 30 30>
      constraintSetup {angularLimitsMax=<0 2 10>, angularLimitsMin=<0 -20 -10>, angularTarget=<1 0 0>}
      localJointOffset=<6 0 0>
     }]
   }
   n72 = Anim Dynamics - Bone: dyn_Hair_Back_03(ComponentPose=@n71, Alpha=1, GravityOverride="0,0,-980", ExternalForce="0,0,0") {
    angularDampingOverride=0.9999
    angularSpringConstant=2000
    bAngularSpring=true
    bOverrideAngularDamping=true
    bUseGravityOverride=true
    gravityScale=0.35
    simulationSpace=RootRelative
    alphaBoolBlend {blendOption=Linear}
    boundBone {boneName=dyn_Hair_Back_03}
    componentAppliedLinearAccClamp=<3000 3000 3000>
    componentLinearAccScale=<1 1 1>
    componentLinearVelScale=<1 1 1>
    gravityOverride {z=-980}
    physicsBodyDefinitions [{
      boundBone {boneName=dyn_Hair_Back_03}
      boxExtents=<20 30 30>
      constraintSetup {angularLimitsMax=<0 2 10>, angularLimitsMin=<0 -20 -10>, angularTarget=<1 0 0>}
      localJointOffset=<6 0 0>
     }]
   }
   n73 = Anim Dynamics - Bone: dyn_Hair_Back_04(ComponentPose=@n72, Alpha=1, GravityOverride="0,0,-980", ExternalForce="0,0,0") {
    angularDampingOverride=0.9999
    angularSpringConstant=2000
    bAngularSpring=true
    bOverrideAngularDamping=true
    bUseGravityOverride=true
    gravityScale=0.35
    simulationSpace=RootRelative
    alphaBoolBlend {blendOption=Linear}
    boundBone {boneName=dyn_Hair_Back_04}
    componentAppliedLinearAccClamp=<3000 3000 3000>
    componentLinearAccScale=<1 1 1>
    componentLinearVelScale=<1 1 1>
    gravityOverride {z=-980}
    physicsBodyDefinitions [{
      boundBone {boneName=dyn_Hair_Back_04}
      boxExtents=<20 30 30>
      constraintSetup {angularLimitsMax=<0 2 10>, angularLimitsMin=<0 -20 -10>, angularTarget=<1 0 0>}
      localJointOffset=<6 0 0>
     }]
   }
   n74 = Anim Dynamics - Bone: dyn_Hair_Back_05(ComponentPose=@n73, Alpha=0, GravityOverride="0,0,-980", ExternalForce="0,0,0") {
    angularDampingOverride=0.9999
    angularSpringConstant=2000
    bAngularSpring=true
    bOverrideAngularDamping=true
    bUseGravityOverride=true
    gravityScale=0.35
    simulationSpace=RootRelative
    alphaBoolBlend {blendOption=Linear}
    boundBone {boneName=dyn_Hair_Back_05}
    componentAppliedLinearAccClamp=<3000 3000 3000>
    componentLinearAccScale=<1 1 1>
    componentLinearVelScale=<1 1 1>
    gravityOverride {z=-980}
    physicsBodyDefinitions [{
      boundBone {boneName=dyn_Hair_Back_05}
      boxExtents=<20 30 30>
      constraintSetup {angularLimitsMax=<0 2 10>, angularLimitsMin=<0 -20 -10>, angularTarget=<1 0 0>}
      localJointOffset=<6 0 0>
     }]
   }
   n75 = Anim Dynamics - Bone: dyn_Hair_Back_06(ComponentPose=@n74, Alpha=1, GravityOverride="0,0,-980", ExternalForce="0,0,0") {
    angularDampingOverride=0.9999
    angularSpringConstant=2000
    bAngularSpring=true
    bOverrideAngularDamping=true
    bUseGravityOverride=true
    gravityScale=0.35
    simulationSpace=RootRelative
    alphaBoolBlend {blendOption=Linear}
    boundBone {boneName=dyn_Hair_Back_06}
    componentAppliedLinearAccClamp=<3000 3000 3000>
    componentLinearAccScale=<1 1 1>
    componentLinearVelScale=<1 1 1>
    gravityOverride {z=-980}
    physicsBodyDefinitions [{
      boundBone {boneName=dyn_Hair_Back_06}
      boxExtents=<20 30 30>
      constraintSetup {angularLimitsMax=<0 2 10>, angularLimitsMin=<0 -20 -10>, angularTarget=<1 0 0>}
      localJointOffset=<6 0 0>
     }]
   }
   n76 = "Transform (Modify) Bone - Bone: dyn_Hair_Back_03"(ComponentPose=@n75, Alpha=1, Translation="0,0,-5", Rotation="-20,0,0", Scale="1,1,1") {
    rotationMode=BMM_Additive
    rotationSpace=BCS_BoneSpace
    translationMode=BMM_Additive
    translationSpace=BCS_BoneSpace
    alphaBoolBlend {blendOption=Linear}
    boneToModify {boneName=dyn_Hair_Back_03}
    rotation {pitch=-20}
    translation {z=-5}
   }
   n77 = "Transform (Modify) Bone - Bone: dyn_Hair_Back_01"(ComponentPose=@n76, Alpha=1, Translation="0,0,0", Rotation="10,0,0", Scale="1,1,1") {
    rotationMode=BMM_Additive
    rotationSpace=BCS_BoneSpace
    translationMode=BMM_Additive
    translationSpace=BCS_BoneSpace
    alphaBoolBlend {blendOption=Linear}
    boneToModify {boneName=dyn_Hair_Back_01}
    rotation {pitch=10}
   }
   Save cached pose 'HairDyn'(Pose=Component To Local(ComponentPose=@n77)) {cacheName=HairDyn}
   Reroute Node(InputPin:Component Space Pose Link Structure=Reroute Node()) -> (OutputPin:Component Space Pose Link Structure)
  ]}
 EventGraph {nodes [
   n1 = Event BlueprintUpdateAnimation()
   > n2 = Sequence!(execute) -> (then_0, then_1, then_2:Exec, then_3:Exec, then_4:Exec, then_5:Exec)
   n3 = Get bIsMoving2D()
   n4 = Get bIsCrouching()
   Set isCrouchWalk(execute=@n2.then_2, isCrouchWalk:double="To Float (Boolean)"(InBool=AND Boolean(A=@n4, B=@n3))) -> (Output_Get:double)
   n5 = GetDeltaTransformFromRefPose(self=GetOwnerSkeletalMeshComponent(), BoneName=head, BaseName=spine_05)
   n6 = MapRangeClamped(Value=BreakRotator(InRot=BreakTransform(InTransform=@n5).Rotation).Yaw, InRangeA=0, InRangeB=-75)
   > Set HeadUp(HeadUp:double=@n6) -> (Output_Get:double)
   > Set CrouchHeadsup(CrouchHeadsup:double="To Float (Boolean)"(InBool=@n4)) -> (Output_Get:double)
   n7 = Get Only_Pistol() -> (Only_Pistol:bool)
   n8 = Get Only_Rifle() -> (Only_Rifle:bool)
   > Set CrouchWalkHeadsup(CrouchWalkHeadsup:double=SelectFloat(A=Get isCrouchWalk() -> (isCrouchWalk:double), B=Get HeadUp() -> (HeadUp:double), bPickA=AND Boolean(A=@n7, B=@n8))) -> (Output_Get:double)
   > Set Parachute_Open_Alpha(Parachute_Open_Alpha:double="To Float (Boolean)"(InBool=Get bIsParachuteOpened())) -> (Output_Get:double)
   n9 = Get Parachute_Open_Alpha() -> (Parachute_Open_Alpha:double)
   n10 = Get isHangGliding()
   > n11 = Set Skydive_float(Skydive_float:double=SelectFloat(A=SelectFloat(A=@n10, B=@n9, bPickA=false), B=Get isSkydiveFloating(), bPickA=false)) -> (Output_Get:double)
   > n12 = Set Skydive_Not(Skydive_Not:bool="Not Equal (Float)"(A=@n11, B=1)) -> (Output_Get:bool)
   > Set Skydive_and_backpack_float_Not(Skydive_and_backpack_float_Not:double="To Float (Boolean)"(InBool=@n12)) -> (Output_Get:double)
   n13 = BreakTransform(InTransform=GetDeltaTransformFromRefPose(self=GetOwnerSkeletalMeshComponent(), BoneName=head, BaseName=spine_05))
   > Set head_rot_back_Extreme(head_rot_back_Extreme:double=MapRangeClamped(Value=BreakRotator(InRot=@n13.Rotation).Yaw, InRangeA=0, InRangeB=-75)) -> (Output_Get:double)
   n14 = Get Only_Rifle()
   n15 = Get bIsCrouching()
   Set Only_Rifle_Alpha(execute=@n2.then_3, Only_Rifle_Alpha:double="To Float (Boolean)"(InBool=AND Boolean(A=@n14, B=@n15))) -> (Output_Get:double)
   n16 = Sequence!(execute=@n2.then_5) -> (then_0, then_1, then_2)
   n17 = Get bIsTargeting()
   n18 = Get CrouchPack() -> (CrouchPack:bool)
   > Set BackBilingHeadsup(BackBilingHeadsup:double=SelectFloat(A=Get HeadUp(), B=0, bPickA=AND Boolean(A=@n18, B=@n17))) -> (Output_Get:double)
   n19 = Get CrouchPack()
   n20 = Get bIsMoving2D()
   > Set BackBilingCrouchWalk(BackBilingCrouchWalk:double="To Float (Boolean)"(InBool=AND Boolean(A=@n20, B=@n19))) -> (Output_Get:double)
   n21 = NOT Boolean(A=Get bIsPlayingEmote())
   n22 = Get bIsInFrontEnd()
   > Set FrontEndidleAlpha(FrontEndidleAlpha:double="To Float (Boolean)"(InBool=AND Boolean(A=@n22, B=@n21))) -> (Output_Get:double)
   n23 = NOT Boolean(A=Get bIsPlayingEmote())
   n24 = Get bIsInFrontEnd()
   n25 = Get isWearingPackBool()
   > Set FrontEndBackbilingAlpha(FrontEndBackbilingAlpha:double="To Float (Boolean)"(InBool=AND Boolean!(A=@n24, B=@n23, C:bool=@n25) -> (ReturnValue))) -> (Output_Get:double)
   n26 = GetAnimInstance(self=GetOwnerSkeletalMeshComponent())
   Is Valid(exec=@n16.then_1, InputObject=@n26)
   > Set isFrontEndIdle(isFrontEndIdle:bool=Montage_IsPlaying(self=@n26, Montage=/BRCosmetics/Animation/Game/MainPlayer/Menu/BR/CustomSkins/_s20/LittleEgg/LittleEgg_Female_Idle_ALT_Montage)) -> (Output_Get:bool)
   n27 = NOT Boolean(A=Get bRigidBodyEnabled())
   n28 = OR Boolean(A=Get bRigidBodyEnabled(), B=@n27)
   n29 = Get isWearingPackBool()
   Set AltSkin_adjust(execute=@n16.then_2, AltSkin_adjust:double=SelectFloat(A=1, B=0, bPickA=AND Boolean!(A=@n28, B=@n29) -> (ReturnValue))) -> (Output_Get:double)
   n30 = "Gravity Override State Function (Parameter Asset) (Nativized)"(
    execute=@n2.then_0
    CharacterPartAnimInstance=Self-Reference()
    GameTimeInSeconds=GetGameTimeInSeconds()
    Parameters=&/F_MED_MysticFang_Hair_FaceAcc_Parameters
   )
   > Set Gravity Override(Gravity Override=@n30)
   n31 = Get isWearingPackBool()
   n32 = Get bIsDBNO()
   > Set DBNO_Cape_Float(DBNO_Cape_Float:double=SelectFloat(A=1, B=0.2, bPickA=OR Boolean(A=@n31, B=@n32))) -> (Output_Get:double)
   > Set DivePack(DivePack:bool=OR Boolean!(A=Get bIsSkydiveDiveMode(), B=Get bIsSkydiveFloating(), C:bool=Get isWearingPackBool(), D:bool=Get bIsDBNO()) -> (ReturnValue)) -> (Output_Get:bool)
   > Set CrouchPack(CrouchPack:bool=OR Boolean!(A=Get bIsCrouching(), B=Get isWearingPackBool()) -> (ReturnValue)) -> (Output_Get:bool)
   > Set FE_Skydive(FE_Skydive:bool=OR Boolean(A=Get bIsSkyDiving(), B=Get bIsInFrontEnd())) -> (Output_Get:bool)
   n33 = Get bIsJumping()
   n34 = Get bIsFalling()
   > Set isTacticalJump(isTacticalJump:bool=AND Boolean(A=Get bIsTacticalSprinting(), B=OR Boolean(A=@n33, B=@n34))) -> (Output_Get:bool)
   > Set isTacticalSprintJump(isTacticalSprintJump:bool=OR Boolean(A=Get isTacticalJump() -> (isTacticalJump:bool), B=Get bIsTacticalSprinting())) -> (Output_Get:bool)
   n35 = OR Boolean(A=Get bIsSkydiveFloating(), B=Get bIsSkydiveDiveMode())
   Set isSkydiveFloating(
    execute=@n2.then_1
    isSkydiveFloating:double=FInterpTo(Current=Get isSkydiveFloating(), Target="To Float (Boolean)"(InBool=@n35), DeltaTime=@n1.DeltaTimeX, InterpSpeed=8)
   ) -> (Output_Get:double)
   n36 = Get bIsCrouching()
   > Set isCrouching(isCrouching:double=FInterpTo(Current=Get isCrouching(), Target="To Float (Boolean)"(InBool=@n36), DeltaTime=@n1.DeltaTimeX, InterpSpeed=15)) -> (Output_Get:double)
   n37 = Get bIsTargeting()
   > Set isTargeting(isTargeting:double=FInterpTo(Current=Get isTargeting(), Target="To Float (Boolean)"(InBool=@n37), DeltaTime=@n1.DeltaTimeX, InterpSpeed=15)) -> (Output_Get:double)
   > n38 = Sequence!(execute) -> (then_0, then_1)
   > n39 = Branch(Condition=Get bIsInFrontEnd())
   > Set BkHairDynAlpha(BkHairDynAlpha:double=0.15) -> (Output_Get:double)
   n40 = Branch(execute=@n39.else, Condition=OR Boolean!(A=Get bIsSkydiveDiveMode(), B=Get bIsSkydiveFloating(), C:bool=Get bIsCrouching()) -> (ReturnValue))
   > Set BkHairDynAlpha(BkHairDynAlpha=0.3)
   Set BkHairDynAlpha(execute=@n40.else, BkHairDynAlpha=0.8)
   n41 = Branch(execute=@n38.then_1, Condition=Get bIsInFrontEnd())
   > Set BangsDynAlpha(BangsDynAlpha:double=0.5) -> (Output_Get:double)
   > Set HatDynAlpha(HatDynAlpha:double=0.25) -> (Output_Get:double)
   > Set FnHairDynAlpha(FnHairDynAlpha:double=0.35) -> (Output_Get:double)
   Set BangsDynAlpha(execute=@n41.else, BangsDynAlpha=1)
   > Set HatDynAlpha(HatDynAlpha=1)
   > n42 = Branch(Condition=Get bIsParachuteOpened())
   > Set FnHairDynAlpha(FnHairDynAlpha=0.5)
   n43 = NOT Boolean(A=Get bIsOnGround() -> (bIsOnGround:bool))
   n44 = Get bIsFalling()
   n45 = Get bIsMoving2D()
   n46 = Get bIsJumping()
   n47 = Get bIsTacticalSprinting()
   > Set Jumping_Alpha(Jumping_Alpha:double="To Float (Boolean)"(InBool=AND Boolean!(A=@n44, B=@n46, C:bool=@n47, D:bool=@n43, E:bool=@n45) -> (ReturnValue))) -> (Output_Get:double)
   Set FnHairDynAlpha(execute=@n42.else, FnHairDynAlpha=1)
   Set Head Down(
    execute=@n2.then_4
    Head Down:double=GetAlphaFromReferenceTransformDelta(SkeletalMeshComponent=Get Skel_Mesh(), BoneName=spine_05, BaseName="head ", InputRange=(X=-90,Y=0), TwistAxis="0,0,-1")
   ) -> (Output_Get:double)
   Event OnParachuteClosed()
   > Set isHangGliding(isHangGliding:double=0) -> (Output_Get:double)
   n48 = Event OnCharacterCustomizationLoaded()
   > Set Skel_Mesh(Skel_Mesh=@n48.CustomBaseMesh)
   > n49 = Cast To PlayerPawn_Athena(Object=TryGetPawnOwner())
   > Set Body_Mesh(Body_Mesh=GetSkeletalMeshForPartType(self=@n49))
   n50 = CheckForBackbling()
   > Set isWearingBackbling(isWearingBackbling=@n50)
   > Set isWearingPackBool(isWearingPackBool:bool=OR Boolean(A=@n50, B=@n50.bIsBackblingACape, C=@n50.bIsBackblingLong)) -> (Output_Get:bool)
   n51 = OR Boolean!(A=@n50.bIsBackblingACape, B=@n50.bIsBackblingLong) -> (ReturnValue)
   > Set is_WearingCape(is_WearingCape:bool=@n51) -> (Output_Get:bool)
   > Set is_f_WearingCape(is_f_WearingCape:double="To Float (Boolean)"(InBool=@n51)) -> (Output_Get:double)
   > Set isWearingBP(isWearingBP:bool=AND Boolean!(A=@n50, B=NOT Boolean(A=@n51)) -> (ReturnValue)) -> (Output_Get:bool)
   > Set Backbling(Backbling:double="To Float (Boolean)"(InBool=OR Boolean!(A=@n50, B=@n50.bIsBackblingACape, C:bool=@n50.bIsBackblingLong) -> (ReturnValue))) -> (Output_Get:double)
   > n52 = Branch(Condition=Get bIsInFrontEnd())
   > Set Sim Space A(Sim Space A:AnimPhysSimSpaceType Enum=World) -> (Output_Get:AnimPhysSimSpaceType Enum)
   > Set Sim Space B(Sim Space B:AnimPhysSimSpaceType Enum=World) -> (Output_Get:AnimPhysSimSpaceType Enum)
   > Set Sim Space C(Sim Space C:AnimPhysSimSpaceType Enum=World) -> (Output_Get:AnimPhysSimSpaceType Enum)
   > Set Sim Space RB(Sim Space RB:Sim Space Settings Structure=Get Sim Space FE() -> (Sim Space FE:Sim Space Settings Structure)) -> (Output_Get:Sim Space Settings Structure)
   Set Sim Space A(execute=@n52.else)
   > Set Sim Space B(Sim Space B=BoneRelative)
   > Set Sim Space C(Sim Space C=Component)
   > Set Sim Space RB(Sim Space RB=Get Sim Space BR() -> (Sim Space BR:Sim Space Settings Structure))
   Event OnParachuteOpened()
   > Switch on EGliderType(Selection=Get PlayerGliderType() -> (PlayerGliderType=HangGlider))
   > Set isHangGliding(isHangGliding=1)
   Event OnWeaponChanged()
   > Set isGunOut(isGunOut:double=1) -> (Output_Get:double)
   > n53 = Switch on EFortWeaponCoreAnimation!(execute, Selection:EFortWeaponCoreAnimation Enum=Get WeaponCoreAnim!() -> (WeaponCoreAnim)) -> (
    Melee:Exec
    Pistol:Exec
    Shotgun:Exec
    PaperBlueprint:Exec
    Rifle:Exec
    MeleeOneHand:Exec
    MachinePistol:Exec
    RocketLauncher:Exec
    GrenadeLauncher:Exec
    GoingCommando:Exec
    AssaultRifle:Exec
    TacticalShotgun:Exec
    SniperRifle:Exec
    TrapPlacement:Exec
    ShoulderLauncher:Exec
    AbilityDecoTool:Exec
    Crossbow:Exec
    C4:Exec
    RemoteControl:Exec
    DualWield:Exec
    AR_BullPup:Exec
    AR_ForwardGrip:Exec
    MedPackPaddles:Exec
    SMG_P90:Exec
    AR_DrumGun:Exec
    Consumable_Small:Exec
    Consumable_Large:Exec
    Balloon:Exec
    MountedTurret:Exec
    CreativeTool:Exec
    ExplosiveBow:Exec
    AshtonIndigo:Exec
    AshtonChicago:Exec
    MeleeDualWield:Exec
    Unarmed:Exec
   )
   Set Rifle(execute=[@n53.AssaultRifle, @n53.Rifle, @n53.SniperRifle], Rifle:double=1) -> (Output_Get:double)
   > Set Only_Rifle(Only_Rifle:bool=1) -> (Output_Get:bool)
   > Set Only_Pistol(Only_Pistol:bool=false) -> (Output_Get:bool)
   Set Melee(execute=[@n53.DualWield, @n53.Melee, @n53.MeleeOneHand], Melee:double=1) -> (Output_Get:double)
   > Set Only_Rifle(Only_Rifle=false)
   > Set Only_Pistol(Only_Pistol=false)
   Set pistol(execute=[@n53.MachinePistol, @n53.Pistol], pistol:double=1) -> (Output_Get:double)
   > Set Only_Rifle(Only_Rifle=false)
   > Set Only_Pistol(Only_Pistol=true)
   Set isGunOut(execute=[@n53.MeleeDualWield, @n53.Unarmed], isGunOut=0)
   SelectFloat(A=@n6, B=0, bPickA=false)
   AND Boolean(A=@n12, B=Get isWearingBackbling())
   OR Boolean!(A=Get bIsInFrontEnd(), B=@n27) -> (ReturnValue)
   Get is_WearingCape() -> (is_WearingCape:bool)
   Get isWearingPackBool()
   Get bIsJumping()
  ]}
]
variables [
 AltSkin_adjust {type=number}
 BackBilingCrouchWalk {type=number}
 BackBilingHeadsup {type=number}
 Backbling {type=number}
 BangsDynAlpha {type=number}
 BkHairDynAlpha {type=number}
 Body_Mesh {type=SkeletalMeshComponent}
 CrouchHeadsup {type=number}
 CrouchPack {type=boolean}
 CrouchWalkHeadsup {type=number}
 DBNO_Cape_Float {type=number}
 DivePack {type=boolean}
 FE_Skydive {type=boolean}
 FnHairDynAlpha {type=number}
 FrontEndBackbilingAlpha {type=number}
 FrontEndidleAlpha {type=number}
 Gravity Override {type=Vector}
 Gravity Override Params {type=GravityOverrideParamsStruct}
 HatDynAlpha {type=number}
 Head Down {type=number}
 head_rot_back_Extreme {type=number}
 HeadUp {type=number}
 is_f_WearingCape {type=number}
 is_WearingCape {type=boolean}
 isCrouching {type=number}
 isCrouchWalk {type=number}
 isFrontEndIdle {type=boolean}
 isGunOut {type=number}
 isHangGliding {type=number}
 isSkydiveFloating {type=number}
 isTacticalJump {type=boolean}
 isTacticalSprintJump {type=boolean}
 isTargeting {type=number}
 isWearingBackbling {type=boolean}
 isWearingBP {type=boolean}
 isWearingPackBool {type=boolean}
 Jumping_Alpha {type=number}
 Melee {type=number}
 Only_Pistol {type=boolean}
 Only_Rifle {type=boolean}
 Only_Rifle_Alpha {type=number}
 Parachute_Open_Alpha {type=number}
 pistol {type=number}
 Rifle {type=number}
 Sim Space A {type=AnimPhysSimSpaceType}
 Sim Space B {type=AnimPhysSimSpaceType}
 Sim Space BR {
  type=SimSpaceSettings
  default {
   dampingAlpha=1
   maxAngularAcceleration=10000
   maxAngularVelocity=10000
   maxLinearAcceleration=10000
   maxLinearVelocity=10000
   velocityScaleZ=1
   worldAlpha=0
   externalLinearDragV=<0 0 0>
  }
 }
 Sim Space C {type=AnimPhysSimSpaceType}
 Sim Space FE {
  type=SimSpaceSettings
  default {
   dampingAlpha=1
   maxAngularAcceleration=10000
   maxAngularVelocity=10000
   maxLinearAcceleration=10000
   maxLinearVelocity=10000
   velocityScaleZ=1
   worldAlpha=0
   externalLinearDragV=<0 0 0>
  }
 }
 Sim Space RB {
  type=SimSpaceSettings
  default {
   dampingAlpha=1
   maxAngularAcceleration=10000
   maxAngularVelocity=10000
   maxLinearAcceleration=10000
   maxLinearVelocity=10000
   velocityScaleZ=1
   worldAlpha=0
   externalLinearDragV=<0 0 0>
  }
 }
 Skel_Mesh {type=SkeletalMeshComponent}
 Skydive_and_backpack_float_Not {type=number}
 Skydive_float {type=number}
 Skydive_Not {type=boolean}
]
