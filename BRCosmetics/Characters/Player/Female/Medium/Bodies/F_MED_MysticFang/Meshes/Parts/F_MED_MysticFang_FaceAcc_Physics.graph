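# Physics asset for the face accessory part of the MysticFang female medium body cosmetic, covering a hat bone and the upper spine.
# One capsule body sits on dyn_hat_bk01, whose name suggests a dynamic back-of-hat bone, and another on spine_05.
# A single constraint joins the two bodies with both swing axes and twist locked, so it allows no relative rotation between them.
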PhysicsAsset /BRCosmetics/Characters/Player/Female/Medium/Bodies/F_MED_MysticFang/Meshes/Parts/F_MED_MysticFang_FaceAcc_Physics
bodies {
 dyn_hat_bk01 [{
   shapeType=Capsule
   length=24.0114
   radius=5.3664
   shapeName=dyn_hat_bk01_capsule
   center=<8.2228 0.5152 4.3716>
   rotation=<rot 0 1.1633 -88.6021>
  }]
 spine_05 [{shapeType=Capsule, length=27.88, radius=7.755, shapeName=spine_05_capsule, rotation=<rot 0 -90 0>}]
}
constraints [{
  bone1Name=dyn_hat_bk01
  bone2Name=spine_05
  swing1LimitDegrees=45
  swing1Motion=Locked
  swing2LimitDegrees=45
  swing2Motion=Locked
  twistLimitDegrees=45
  twistMotion=Locked
 }]
